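# Niagara system for a firework shell that launches upward with trails and shows a double sphere burst.
# The Source particle drives spark, smoke and ribbon-smoke trails through particle reads, alongside a GPU spark-trail emitter, and generates a death event.
# The burst combines a sprite glow with blast and ribbon mesh emitters, each using two renderers and curve-animated material parameters.
# Shared system variables set burst color and size, with the smaller blast hue-shifted.
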NiagaraSystem /SparksEffects/Effects/Fireworks/Particles/NS_Sparks_Firework_Sphere_Double
dependencies {
 usedDataInterfaces [
  &NiagaraDataInterfaceCurve
  &NiagaraDataInterfaceParticleRead
  &NiagaraDataInterfaceColorCurve
  &NiagaraDataInterfaceArrayColor
 ]
 usedDynamicInputs [&SelectLinearColorFromArray]
 usedModules [
  &self:SystemSpawnScript.SystemScriptSource.SystemScriptGraph.NiagaraNodeAssignment_1.SetVariables_#1
  &SystemState
 ]
}
emitters {
 FIrework_Blast_Combined {
  emitterUpdateScript [&EmitterState()]
  particleSpawnScript [&InitializeParticle(
    Color=$System.SystemColor
    Lifetime=4
    Mesh Scale Mode=::NewEnumerator3
    Mesh Uniform Scale=15
    UsePositionOffset=false
    Write Material Random=true
   )]
  particleUpdateScript [
   &ParticleState()
   &ScaleMeshSize(Scale Factor=&VectorFromFloat(Value=$System.BurstMeshSize))
   /SetVariables(Particles.Scale Mesh Size Blast Small:Vector3f=&Multiply_Vector(
     A=&VectorFromFloat(Value=10)
     B=&VectorFromFloat(Value=$System.BurstMeshSize)
    )) {isSetParameters=true}
   &ScaleColor(
    Scale Alpha=&FloatFromCurve(FloatCurve=DataInterface() {propertyValues {Curve {keys [
         {arriveTangent=0.0019, interpMode=RCIM_Cubic, leaveTangent=0.0019, tangentMode=RCTM_User, time=0, value=1}
         {interpMode=RCIM_Cubic, tangentMode=RCTM_User, time=0.5, value=1}
         {arriveTangent=0.0018, interpMode=RCIM_Cubic, leaveTangent=0.0018, tangentMode=RCTM_User, time=0.75, value=0}
        ]}}})
    Scale RGB=&VectorFromFloat(Value=3)
    ScaleA=true
    ScaleRGB=true
   )
   /SetVariables(Particles.Blast Small Color:LinearColor=&MultiplyLinearColor(
     Linear Color=&HueShiftLinearColor(Hue=0.2, Saturation=1, Value=1, Linear Color=$System.SystemColor)
     Scale Factor=&MakeVector4(
      X=2
      Y=2
      Z=2
      W=&FloatFromCurve(FloatCurve=DataInterface() {propertyValues {Curve {keys [
           {arriveTangent=0.0019, interpMode=RCIM_Cubic, leaveTangent=0.0019, tangentMode=RCTM_User, time=0, value=1}
           {interpMode=RCIM_Cubic, tangentMode=RCTM_User, time=0.2, value=1}
           {arriveTangent=0.0018, interpMode=RCIM_Cubic, leaveTangent=0.0018, tangentMode=RCTM_User, time=1.0011, value=0}
          ]}}})
     )
    )) {isSetParameters=true}
   &SolveForcesAndVelocity(Clamp Velocity=false, Limit Acceleration=false)
   &DynamicMaterialParameters(
    Index 0 Param 1=&FloatFromCurve(
     FloatCurve=DataInterface() {propertyValues {Curve {keys [
         {arriveTangent=-0.0112, interpMode=RCIM_Cubic, leaveTangent=-0.0112, tangentMode=RCTM_User, time=0, value=0.4}
         {arriveTangent=-0.7264, interpMode=RCIM_Cubic, leaveTangent=-0.7264, tangentMode=RCTM_User, time=0.1, value=0.3}
         {arriveTangent=-0.1797, interpMode=RCIM_Cubic, leaveTangent=-0.1797, tangentMode=RCTM_User, time=0.4, value=0.225}
         {arriveTangent=-0.1859, interpMode=RCIM_Cubic, leaveTangent=-0.1859, tangentMode=RCTM_User, time=0.6, value=0.175}
         {arriveTangent=-3.4317, interpMode=RCIM_Cubic, leaveTangent=-3.4317, tangentMode=RCTM_User, time=0.75, value=0}
        ]}}}
     Scale Curve=300
    )
    Index 0 Param 2=&FloatFromCurve(FloatCurve=DataInterface() {propertyValues {Curve {keys [
         {arriveTangent=14.9629, interpMode=RCIM_Cubic, leaveTangent=14.9629, tangentMode=RCTM_User, time=0, value=-0.7}
         {arriveTangent=3.0529, interpMode=RCIM_Cubic, leaveTangent=3.0529, tangentMode=RCTM_User, time=0.16, value=0.5}
         {interpMode=RCIM_Cubic, tangentMode=RCTM_User, time=0.75, value=1}
        ]}}})
    Index 0 Param 3=&FloatFromCurve(
     FloatCurve=DataInterface() {propertyValues {Curve {keys [
         {interpMode=RCIM_Cubic, tangentMode=RCTM_User, time=0, value=0}
         {arriveTangent=-856.4041, interpMode=RCIM_Cubic, leaveTangent=-856.4041, tangentMode=RCTM_User, time=0.75, value=-500}
        ]}}}
     Scale Curve=3
    )
    Index 0 Param 4=&FloatFromCurve(
     FloatCurve=DataInterface() {propertyValues {Curve {keys [
         {arriveTangent=0.0932, interpMode=RCIM_Cubic, leaveTangent=0.0932, tangentMode=RCTM_User, time=0, value=0}
         {arriveTangent=1.4953, interpMode=RCIM_Cubic, leaveTangent=1.4953, tangentMode=RCTM_User, time=0.75, value=1}
        ]}}}
     Scale Curve=600
    )
    Index 1 Param 1=-0.1
    Module.Param0WriteEnabled=true
    Module.Param0WriteEnabled1=true
    Module.Param1WriteEnabled=true
    Module.Param1WriteEnabled1=true
    Module.Param2WriteEnabled=true
    Module.Param2WriteEnabled1=true
    Module.Param3WriteEnabled=true
    Module.Param3WriteEnabled1=true
    Write Parameter Index 1=true
   )
   /SetVariables(Particles.Blast Small Dynamic Param 0:Vector4f=&MakeVector4(
     X=&FloatFromCurve(
      FloatCurve=DataInterface() {propertyValues {Curve {keys [
          {arriveTangent=-0.0112, interpMode=RCIM_Cubic, leaveTangent=-0.0112, tangentMode=RCTM_User, time=0, value=0.4}
          {arriveTangent=-0.7264, interpMode=RCIM_Cubic, leaveTangent=-0.7264, tangentMode=RCTM_User, time=0.1, value=0.3}
          {arriveTangent=-0.1797, interpMode=RCIM_Cubic, leaveTangent=-0.1797, tangentMode=RCTM_User, time=0.4326, value=0.2242}
          {arriveTangent=-0.1859, interpMode=RCIM_Cubic, leaveTangent=-0.1859, tangentMode=RCTM_User, time=0.8879, value=0.1764}
          {arriveTangent=-3.4317, interpMode=RCIM_Cubic, leaveTangent=-3.4317, tangentMode=RCTM_User, time=1, value=0}
         ]}}}
      Scale Curve=300
     )
     Y=&FloatFromCurve(FloatCurve=DataInterface() {propertyValues {Curve {keys [
          {arriveTangent=10.0245, interpMode=RCIM_Cubic, leaveTangent=10.0245, tangentMode=RCTM_User, time=0, value=-1}
          {arriveTangent=1.4411, interpMode=RCIM_Cubic, leaveTangent=1.4411, tangentMode=RCTM_User, time=0.3, value=0.375}
          {interpMode=RCIM_Cubic, tangentMode=RCTM_User, time=1, value=0.75}
         ]}}})
     Z=&FloatFromCurve(
      FloatCurve=DataInterface() {propertyValues {Curve {keys [
          {interpMode=RCIM_Cubic, tangentMode=RCTM_User, time=0, value=0}
          {arriveTangent=-856.4041, interpMode=RCIM_Cubic, leaveTangent=-856.4041, tangentMode=RCTM_User, time=1, value=-500}
         ]}}}
      Scale Curve=2
     )
     W=&FloatFromCurve(
      FloatCurve=DataInterface() {propertyValues {Curve {keys [
          {arriveTangent=0.0932, interpMode=RCIM_Cubic, leaveTangent=0.0932, tangentMode=RCTM_User, time=0, value=0}
          {arriveTangent=1.4953, interpMode=RCIM_Cubic, leaveTangent=1.4953, tangentMode=RCTM_User, time=1, value=1}
         ]}}}
      Scale Curve=300
     )
    )) {isSetParameters=true}
   /SetVariables() {isSetParameters=true}
  ]
  rendererClasses [&NiagaraMeshRendererProperties]
  renderers [
   {rendererClass=&NiagaraMeshRendererProperties, rendererIndex=0}
   {rendererClass=&NiagaraMeshRendererProperties, rendererIndex=1}
  ]
 }
 Light_Glow {
  emitterUpdateScript [&EmitterState()]
  particleSpawnScript [
   &InitializeParticle(
    Color=$System.SystemColor
    Lifetime=3
    Sprite Size Mode=::NewEnumerator3
    Uniform Sprite Size=20
    UsePositionOffset=false
    Write Material Random=true
   )
   &DynamicMaterialParameters(
    Index 0 Param 1=2.75
    Module.Index0Param0Write=true
    Module.Index0Param1Write=true
    Module.Index0Param2Write=true
    Module.Index0Param3Write=true
   )
  ]
  particleUpdateScript [
   &ParticleState()
   &ScaleSpriteSize(
    Scale Sprite Size Mode=::NewEnumerator0
    Uniform Scale Factor=&Multiply_Float(A=$System.BurstMeshSize, B=1.15)
   ) {module=ScaleSpriteSize001}
   &ScaleSpriteSize(
    Uniform Curve Scale=450
    Uniform Curve Sprite Scale=DataInterface() {propertyValues {Curve {keys [
        {arriveTangent=2.1659, interpMode=RCIM_Cubic, leaveTangent=2.1659, tangentMode=RCTM_User, time=0, value=0.5}
        {interpMode=RCIM_Cubic, tangentMode=RCTM_User, time=1, value=1}
       ]}}}
   )
   &ScaleColor(
    Scale Alpha=&FloatFromCurve(FloatCurve=DataInterface() {propertyValues {Curve {keys [
         {interpMode=RCIM_Cubic, tangentMode=RCTM_User, time=0.2, value=1}
         {arriveTangent=-1.7914, interpMode=RCIM_Cubic, leaveTangent=-1.7914, tangentMode=RCTM_User, time=1, value=0}
        ]}}})
    Scale RGB=&VectorFromFloat(Value=1.5)
    ScaleA=true
    ScaleRGB=true
   )
   &GravityForce(Gravity=<0 0 -100>)
   &SolveForcesAndVelocity(Clamp Velocity=false, Limit Acceleration=false)
  ]
  rendererClasses [&NiagaraSpriteRendererProperties]
  renderers [{rendererClass=&NiagaraSpriteRendererProperties}]
 }
 Mesh_Ribbons_Combined {
  emitterUpdateScript [&EmitterState() {module=EmitterState001}]
  particleSpawnScript [
   &InitializeParticle(
    Color=$System.SystemColor
    Lifetime=&RandomRangeFloat(Minimum=2.5, Maximum=3)
    Mesh Scale Mode=::NewEnumerator3
    Sprite Size Mode=::NewEnumerator3
    Uniform Sprite Size=50
    UsePositionOffset=false
    Write Material Random=true
   )
   &InitialMeshOrientation(Use Rotation Vector=false)
   /SetVariables(Particles.Ribbon No Sparks Orientation:Quat4f=/Niagara/DynamicInputs/Rotation/RandomQuaternion()) {isSetParameters=true}
  ]
  particleUpdateScript [
   &ParticleState()
   &ScaleMeshSize(Scale Factor=&VectorFromFloat(Value=$System.BurstMeshSize))
   &ScaleColor(
    Scale Alpha=&FloatFromCurve(FloatCurve=DataInterface() {propertyValues {Curve {keys [
         {arriveTangent=0.0543, interpMode=RCIM_Cubic, leaveTangent=0.0543, tangentMode=RCTM_User, time=0, value=0}
         {arriveTangent=0.0595, interpMode=RCIM_Cubic, leaveTangent=0.0595, tangentMode=RCTM_User, time=1, value=1}
        ]}}})
    Scale RGB=&VectorFromFloat(Value=&FloatFromCurve(FloatCurve=DataInterface() {propertyValues {Curve {keys [
          {arriveTangent=-33.3256, interpMode=RCIM_Cubic, leaveTangent=-33.3256, tangentMode=RCTM_User, time=0, value=5}
          {arriveTangent=-0.2459, interpMode=RCIM_Cubic, leaveTangent=-0.2459, tangentMode=RCTM_User, time=0.25, value=1}
         ]}}}))
    ScaleA=true
    ScaleRGB=true
   )
   &DynamicMaterialParameters(
    Index 0 Param 1=&FloatFromCurve(
     FloatCurve=DataInterface() {propertyValues {Curve {keys [
         {arriveTangent=4.4937, interpMode=RCIM_Cubic, leaveTangent=4.4937, tangentMode=RCTM_User, time=0, value=0}
         {arriveTangent=1.4506, interpMode=RCIM_Cubic, leaveTangent=1.4506, tangentMode=RCTM_User, time=0.2, value=0.625}
         {arriveTangent=0.0378, interpMode=RCIM_Cubic, leaveTangent=0.0378, tangentMode=RCTM_User, time=1, value=1}
        ]}}}
     Scale Curve=3
    )
    Index 0 Param 2=&FloatFromCurve(
     FloatCurve=DataInterface() {propertyValues {Curve {keys [
         {arriveTangent=-0.018, interpMode=RCIM_Cubic, leaveTangent=-0.018, tangentMode=RCTM_User, time=0, value=0}
         {arriveTangent=1.4018, interpMode=RCIM_Cubic, leaveTangent=1.4018, tangentMode=RCTM_User, time=1, value=1}
        ]}}}
     Scale Curve=2
    )
    Index 0 Param 3=-0.2
    Index 1 Param 1=1
    Index 1 Param 2=&FloatFromCurve(
     FloatCurve=DataInterface() {propertyValues {Curve {keys [
         {arriveTangent=-6.499, interpMode=RCIM_Cubic, leaveTangent=-6.499, tangentMode=RCTM_User, time=0, value=3}
         {interpMode=RCIM_Cubic, tangentMode=RCTM_User, time=1, value=1}
        ]}}}
     Scale Curve=3
    )
    Index 1 Param 3=0.5
    Index 1 Param 4=&FloatFromCurve(
     FloatCurve=DataInterface() {propertyValues {Curve {keys [
         {arriveTangent=0.0351, interpMode=RCIM_Cubic, leaveTangent=0.0351, tangentMode=RCTM_User, time=0, value=0}
         {interpMode=RCIM_Cubic, tangentMode=RCTM_User, time=0.7, value=1}
         {interpMode=RCIM_Cubic, tangentMode=RCTM_User, time=1, value=0}
        ]}}}
     Scale Curve=125
    )
    Index 2 Param 1=1
    Index 2 Param 2=&FloatFromCurve(
     FloatCurve=DataInterface() {propertyValues {Curve {keys [
         {arriveTangent=0.0495, interpMode=RCIM_Cubic, leaveTangent=0.0495, tangentMode=RCTM_User, time=0, value=0}
         {arriveTangent=1.1556, interpMode=RCIM_Cubic, leaveTangent=1.1556, tangentMode=RCTM_User, time=1, value=1}
        ]}}}
     Scale Curve=5
    )
    Index 2 Param 3=1
    Index 2 Param 4=0.35
    Module.Param0WriteEnabled=true
    Module.Param0WriteEnabled1=true
    Module.Param0WriteEnabled2=true
    Module.Param1WriteEnabled=true
    Module.Param1WriteEnabled1=true
    Module.Param1WriteEnabled2=true
    Module.Param2WriteEnabled=true
    Module.Param2WriteEnabled1=true
    Module.Param2WriteEnabled2=true
    Module.Param3WriteEnabled=true
    Module.Param3WriteEnabled1=true
    Module.Param3WriteEnabled2=true
    Write Parameter Index 1=true
    Write Parameter Index 2=true
   )
   /SetVariables(Particles.Ribbon No Sparks Params 0:Vector4f=&MakeVector4(
     X=&FloatFromCurve(
      FloatCurve=DataInterface() {propertyValues {Curve {keys [
          {arriveTangent=4.4937, interpMode=RCIM_Cubic, leaveTangent=4.4937, tangentMode=RCTM_User, time=0, value=0}
          {arriveTangent=1.4506, interpMode=RCIM_Cubic, leaveTangent=1.4506, tangentMode=RCTM_User, time=0.2, value=0.625}
          {arriveTangent=0.0378, interpMode=RCIM_Cubic, leaveTangent=0.0378, tangentMode=RCTM_User, time=1, value=1}
         ]}}}
      Scale Curve=2
     )
     Y=&FloatFromCurve(
      FloatCurve=DataInterface() {propertyValues {Curve {keys [
          {arriveTangent=-0.018, interpMode=RCIM_Cubic, leaveTangent=-0.018, tangentMode=RCTM_User, time=0, value=0}
          {arriveTangent=1.4018, interpMode=RCIM_Cubic, leaveTangent=1.4018, tangentMode=RCTM_User, time=1, value=1}
         ]}}}
      Scale Curve=6
     )
     Z=0
     W=0
    )) {isSetParameters=true}
   /SetVariables(Particles.Ribbon No Sparks Params 1:Vector4f=&MakeVector4(
     X=1
     Y=&FloatFromCurve(
      FloatCurve=DataInterface() {propertyValues {Curve {keys [
          {arriveTangent=-6.499, interpMode=RCIM_Cubic, leaveTangent=-6.499, tangentMode=RCTM_User, time=0, value=3}
          {interpMode=RCIM_Cubic, tangentMode=RCTM_User, time=1, value=1}
         ]}}}
      Scale Curve=2
     )
     Z=0.5
     W=&FloatFromCurve(
      FloatCurve=DataInterface() {propertyValues {Curve {keys [
          {arriveTangent=29.4837, interpMode=RCIM_Cubic, leaveTangent=29.4837, tangentMode=RCTM_User, time=0, value=0}
          {interpMode=RCIM_Cubic, tangentMode=RCTM_User, time=0.1, value=1}
          {interpMode=RCIM_Cubic, tangentMode=RCTM_User, time=0.7, value=1}
          {interpMode=RCIM_Cubic, tangentMode=RCTM_User, time=1, value=0}
         ]}}}
      Scale Curve=10
     )
    )) {isSetParameters=true}
   /SetVariables(Particles.Ribbon No Sparks Params 2:Vector4f=&MakeVector4(
     X=1
     Y=&FloatFromCurve(
      FloatCurve=DataInterface() {propertyValues {Curve {keys [
          {arriveTangent=0.8935, interpMode=RCIM_Cubic, leaveTangent=0.8935, tangentMode=RCTM_User, time=0.1, value=0}
          {arriveTangent=1.1556, interpMode=RCIM_Cubic, leaveTangent=1.1556, tangentMode=RCTM_User, time=1, value=1}
         ]}}}
      Scale Curve=5
     )
     Z=1
     W=0.1
    )) {isSetParameters=true}
  ]
  rendererClasses [&NiagaraMeshRendererProperties]
  renderers [
   {rendererClass=&NiagaraMeshRendererProperties, rendererIndex=0}
   {rendererClass=&NiagaraMeshRendererProperties, rendererIndex=1}
  ]
 }
 Ribbon_Smoke {
  emitterUpdateScript [
   &EmitterState(
    Enable Distance Culling=true
    Life Cycle Mode=::NewEnumerator1
    Loop Behavior=::NewEnumerator1
    MaxDistance=2500000
    Module.bUseMaxDistance=true
    Module.bUseMinDistance=false
    Scalability Mode=::NewEnumerator1
    UseLoopDelay=false
   )
   &SpawnParticlesFromOtherEmitter(
    Attribute Reader=DataInterface() {propertyValues {EmitterBinding {bindingMode=Other, emitterName=Source}}}
    Spawn Rate=&FloatFromCurve(FloatCurve=DataInterface() {propertyValues {Curve {keys [<at 1.6 1>,<at 4 0>]}}}, CurveIndex=$Emitter.Age, Scale Curve=15)
    Spawn Rate Per Particle Cap=200
    Use Maximum Age=false
    Use Minimum Age=false
   )
  ]
  particleSpawnScript [
   &InitializeRibbon(
    Position Mode:ENiagara_PositionInitializationMode=::NewEnumerator1
    Ribbon Width=&FloatFromCurve(FloatCurve=DataInterface() {propertyValues {Curve {keys [<at 1 1 RCIM_Cubic>,<at 2 0>]}}}, CurveIndex=$Emitter.Age, Scale Curve=35)
    UsePositionOffset=false
    Write Material Random=true
   ) {module=InitializeRibbon001}
   &SampleParticlesFromOtherEmitter(
    Age Sampling:ENiagara_DirectReadSamplingMode=::NewEnumerator0
    Applied Velocity Scale=0
    Color Sampling=::NewEnumerator0
    Lifetime Application:ENiagara_LifetimeInheritanceOptions=::NewEnumerator1
    Lifetime Sampling:ENiagara_DirectReadSamplingMode=::NewEnumerator0
    Particle ID Sampling=::NewEnumerator1
    Reject Particle If Sampled Index Is Invalid=false
    Sprite Size Sampling:ENiagara_DirectReadSamplingMode=::NewEnumerator0
    Velocity Sampling=::NewEnumerator0
   )
   &AddVelocity(Velocity=&RandomRangeVector(Minimum=<0 -5 -5>, Maximum=<-1 5 5>), Velocity Speed Scale=17)
   &ApplyInitialForces()
  ]
  particleUpdateScript [
   &ParticleState()
   &DynamicMaterialParameters(
    Index 0 Param 1=-0.2
    Index 0 Param 2=1
    Index 0 Param 4=&FloatFromCurve(FloatCurve=DataInterface() {propertyValues {Curve {keys [
         <at 0 1 RCIM_Cubic>
         {arriveTangent=-0.3679, interpMode=RCIM_Cubic, leaveTangent=-0.3679, tangentMode=RCTM_User, time=1.6, value=0.7}
         {arriveTangent=-0.0204, tangentMode=RCTM_User, time=3, value=0}
        ]}}})
    Module.Param0WriteEnabled=true
    Module.Param1WriteEnabled=true
    Module.Param2WriteEnabled=true
    Module.Param3WriteEnabled=true
   )
   &Color(Color=&HueShiftLinearColor(
     Hue=0
     Saturation=1
     Value=1
     Linear Color=&ColorFromCurve(ColorCurve=DataInterface() {propertyValues {
        AlphaCurve {keys [<at 0 1 RCIM_Cubic>,<at 0.3828 1>,<at 1 0 RCIM_Cubic>]}
        BlueCurve {keys [<at 0 0.0469>, {arriveTangent=-0.1346, interpMode=RCIM_Cubic, leaveTangent=-0.1346, time=0.0181, value=0}, <at 0.3483 0 RCIM_Cubic>]}
        GreenCurve {keys [<at 0 0.0469>, {arriveTangent=-0.1346, interpMode=RCIM_Cubic, leaveTangent=-0.1346, time=0.0181, value=0}, <at 0.3483 0 RCIM_Cubic>]}
        RedCurve {keys [<at 0 0.0469>, {arriveTangent=-0.1346, interpMode=RCIM_Cubic, leaveTangent=-0.1346, time=0.0181, value=0}, <at 0.3483 0 RCIM_Cubic>]}
       }})
    ))
   &ScaleRibbonWidth(Ribbon Width Scale=&FloatFromCurve(
     FloatCurve=DataInterface() {propertyValues {Curve {keys [
         {arriveTangent=0.0097, interpMode=RCIM_Cubic, leaveTangent=0.0097, tangentMode=RCTM_User, time=0.1, value=0.5}
         {arriveTangent=0.324, interpMode=RCIM_Cubic, leaveTangent=0.324, tangentMode=RCTM_User, time=0.9, value=0.875}
         <at 2 1>
        ]}}}
     Scale Curve=8
    ))
   &SolveForcesAndVelocity(Clamp Velocity=false, Limit Acceleration=false)
  ]
  rendererClasses [&NiagaraRibbonRendererProperties]
  renderers [{rendererClass=&NiagaraRibbonRendererProperties}]
 }
 Source {
  emitterUpdateScript [
   &EmitterState()
   &SpawnBurst_Instantaneous(Use Spawn Probability=false, UseLoopCountLimit=false)
  ]
  particleSpawnScript [
   &InitializeParticle(
    Color=&MultiplyLinearColor(Linear Color=<rgba 1.5 0.4362 0 1>, Scale Factor=<10 10 10 1>)
    Lifetime=$System.LifetimeRand
    Mass=400
    Mass Mode=::NewEnumerator1
    Position Offset=<0 0 -700>
    Sprite Size Mode=::NewEnumerator3
    Uniform Sprite Size=15000
    UsePositionOffset=true
    Write Material Random=true
   )
   /Niagara/Modules/Update/Mass/CalculateSizeAndRotationalInertiaByMass(Module.Manually Enter Density=false)
   &AddVelocity(Velocity=<0 0 12000>) {module=AddVelocity001}
   &ApplyInitialForces()
   /SetVariables(Particles.SharpFlareColor:LinearColor=<rgba 25 7.2697 0 1>, Particles.SharpFlareSize2:Vector2f=<400 150>) {isSetParameters=true}
  ]
  particleUpdateScript [
   &ParticleState()
   /SetVariables(
    Particles.SharpFlareColor:LinearColor=&MultiplyLinearColor(
     Linear Color=$Particles.SharpFlareColor
     Scale Factor=&MakeVector4(
      X=1
      Y=1
      Z=1
      W=&FloatFromCurve(FloatCurve=DataInterface() {propertyValues {Curve {keys [
           <at 0 1>
           {arriveTangent=-1, interpMode=RCIM_Cubic, leaveTangent=-1, time=0.6, value=1.05}
           {arriveTangent=-4.0049, interpMode=RCIM_Cubic, leaveTangent=-4.0049, tangentMode=RCTM_User, time=1, value=0}
          ]}}})
     )
    )
    Particles.SharpFlareSize2:Vector2f=$Particles.SharpFlareSize2
   ) {isSetParameters=true}
   &ScaleColor(
    Scale Alpha=&FloatFromCurve(
     FloatCurve=DataInterface() {propertyValues {Curve {keys [
         {arriveTangent=9.0725, interpMode=RCIM_Cubic, leaveTangent=9.0725, tangentMode=RCTM_User, time=0, value=0}
         {arriveTangent=-7.2695, interpMode=RCIM_Cubic, leaveTangent=-7.2695, tangentMode=RCTM_User, time=0.2, value=1.75}
         <at 0.9 1.625>
         {arriveTangent=-0.0405, interpMode=RCIM_Cubic, leaveTangent=-0.0405, tangentMode=RCTM_User, time=1, value=0}
        ]}}}
     Scale Curve=0.1
    )
    Scale RGB=&Multiply_Vector(
     A=<1.5 1.5 1.5>
     B=&VectorFromFloat(Value=&FloatFromCurve(
       FloatCurve=DataInterface() {propertyValues {Curve {keys [<at 0 1>, {interpMode=RCIM_Cubic, tangentMode=RCTM_User, time=0.9, value=0.85}, <at 1 0>]}}}
       Scale Curve=&Multiply_Float(
        A=1
        B=&Waveform(
         Waveform Count=::NewEnumerator1
         "Waveform [1]"=::NewEnumerator0
         "[1] Phase Scale"=&RandomRangeFloat(Minimum=3, Maximum=10)
         "[1] Frequency"=1.25
         "[1] Amplitude Min/Max"=<0.25 1>
         "[2] Average"=false
         "[2] Input Position"=0
         "[2] Phase"=$Emitter.Age
         "[2] Phase Scale"=5
         "[2] Compound Frequency"=2
         "[2] Frequency"=2
         "[2] Amplitude Min/Max"=<0.25 1>
         "[2] Scale"=1
         "[2] Bias"=0
         "[2] Exponent"=1
         "[2] Pulse Duration"=0.1
         "[2] Clip Triangle Min/Max"=<0 0>
         "[2] Random Seed"=0
        )
       )
      ))
    )
    ScaleA=true
    ScaleRGB=true
   ) {module=ScaleColor001}
   &ScaleSpriteSize(Uniform Curve Sprite Scale=DataInterface() {propertyValues {Curve {keys [
        {arriveTangent=2.9868, interpMode=RCIM_Cubic, leaveTangent=2.9868, tangentMode=RCTM_User, time=0, value=1.125}
        <at 0.4 2.3>
        <at 0.8 2.3>
        {arriveTangent=-0.0405, interpMode=RCIM_Cubic, leaveTangent=-0.0405, tangentMode=RCTM_User, time=1, value=2}
       ]}}})
   &DynamicMaterialParameters(
    Index 0 Param 2=2
    Index 0 Param 3=1
    Index 1 Param 1=150
    Index 1 Param 4=1
    Module.Param0WriteEnabled=true
    Module.Param0WriteEnabled1=true
    Module.Param1WriteEnabled=true
    Module.Param1WriteEnabled1=true
    Module.Param2WriteEnabled=true
    Module.Param2WriteEnabled1=true
    Module.Param3WriteEnabled=false
    Module.Param3WriteEnabled1=true
    Write Parameter Index 1=true
   )
   &Drag(Use Linear Drag=true, Use Rotational Drag=true)
   &GravityForce()
   &SolveForcesAndVelocity(Clamp Velocity=false, Limit Acceleration=false)
   &ScaleSpriteSizeBySpeed(Min Scale Factor=<0.5 0.5>)
   &GenerateDeathEvent(Use Event Probability=false)
   &self:ScratchModule() {enabled=false, module=ScratchModule}
   &ScaleSpriteSize(
    Uniform Curve Scale=2
    Uniform Curve Sprite Scale=DataInterface() {propertyValues {Curve {keys [
        <at 0 1>
        {arriveTangent=-0.5662, interpMode=RCIM_Cubic, leaveTangent=-0.5662, tangentMode=RCTM_User, time=0.6, value=0.9}
        {arriveTangent=-3.0603, interpMode=RCIM_Cubic, leaveTangent=-3.0603, tangentMode=RCTM_User, time=1, value=0}
       ]}}}
   ) {module=ScaleSpriteSize001}
   &CameraOffset(Camera Offset Amount=10)
  ]
  rendererClasses [&NiagaraSpriteRendererProperties]
  renderers [
   {rendererClass=&NiagaraSpriteRendererProperties}
   {rendererClass=&NiagaraSpriteRendererProperties, rendererIndex=1}
  ]
 }
 Source_SparkTrail001 {
  simTarget=GPUComputeSim
  emitterUpdateScript [
   &EmitterState()
   &SpawnBurst_Instantaneous(Use Spawn Probability=false, UseLoopCountLimit=false)
  ]
  particleSpawnScript [
   &InitializeParticle(
    Color=<rgba 1.25 0.3635 0 0.5>
    Lifetime=$System.LifetimeRand
    Position Offset=<0 0 -850>
    Sprite Size=<350 400>
    Sprite Size Mode=::NewEnumerator4
    UsePositionOffset=true
    Write Material Random=true
   )
   &AddVelocity(Velocity=<0 0 12000>)
  ]
  particleUpdateScript [
   &Drag(Use Linear Drag=true, Use Rotational Drag=true)
   &GravityForce()
   &SolveForcesAndVelocity(Clamp Velocity=false, Limit Acceleration=false)
   &ParticleState()
   &DynamicMaterialParameters(
    Index 0 Param 1=0.8
    Index 0 Param 2=1
    Index 0 Param 3=1
    Index 1 Param 1=150
    Index 1 Param 4=1
    Module.Param0WriteEnabled=true
    Module.Param0WriteEnabled1=true
    Module.Param1WriteEnabled=true
    Module.Param1WriteEnabled1=true
    Module.Param2WriteEnabled=true
    Module.Param2WriteEnabled1=true
    Module.Param3WriteEnabled=false
    Module.Param3WriteEnabled1=true
    Write Parameter Index 1=true
   )
   &ScaleColor(
    Scale Alpha=&FloatFromCurve(
     FloatCurve=DataInterface() {propertyValues {Curve {keys [
         {arriveTangent=9.0725, interpMode=RCIM_Cubic, leaveTangent=9.0725, tangentMode=RCTM_User, time=0, value=0}
         {arriveTangent=-7.2695, interpMode=RCIM_Cubic, leaveTangent=-7.2695, tangentMode=RCTM_User, time=0.2, value=1.75}
         <at 0.8 1.8>
         {arriveTangent=-0.0405, interpMode=RCIM_Cubic, leaveTangent=-0.0405, tangentMode=RCTM_User, time=1, value=0}
        ]}}}
     Scale Curve=0.02
    )
    ScaleA=true
    ScaleRGB=true
   ) {module=ScaleColor001}
   &ScaleSpriteSize(
    Uniform Curve Scale=0.15
    Uniform Curve Sprite Scale=DataInterface() {propertyValues {Curve {keys [
        {arriveTangent=2.9868, interpMode=RCIM_Cubic, leaveTangent=2.9868, tangentMode=RCTM_User, time=0, value=1.125}
        <at 0.4 2.3>
        <at 0.8 2.3>
        {arriveTangent=-0.0405, interpMode=RCIM_Cubic, leaveTangent=-0.0405, tangentMode=RCTM_User, time=1, value=1.4}
       ]}}}
   )
   &ScaleSpriteSizeBySpeed(Max Scale Factor=<2 5>, Min Scale Factor=<1.5 0.25>, Velocity Threshold=2500)
   &ScaleSpriteSize(Uniform Curve Sprite Scale=DataInterface() {propertyValues {Curve {keys [
        {interpMode=RCIM_Cubic, tangentMode=RCTM_User, time=0, value=1}
        {arriveTangent=-0.0015, interpMode=RCIM_Cubic, leaveTangent=-0.0015, tangentMode=RCTM_User, time=1, value=0}
       ]}}}) {module=ScaleSpriteSize001}
  ]
  rendererClasses [&NiagaraSpriteRendererProperties]
  renderers [{rendererClass=&NiagaraSpriteRendererProperties}]
 }
 Trail_Smoke {
  emitterUpdateScript [
   &EmitterState()
   &SpawnParticlesFromOtherEmitter(
    Attribute Reader=DataInterface() {propertyValues {EmitterBinding {bindingMode=Other, emitterName=Source}}}
    Spawn Rate=65
    Spawn Rate Per Particle Cap=512
    Use Maximum Age=false
    Use Minimum Age=false
   )
  ]
  particleSpawnScript [
   &InitializeParticle(
    Color=<rgba 0.3958 0.3958 0.3958 0.08>
    Lifetime Max=1.25
    Lifetime Min=0.5
    Lifetime Mode=::NewEnumerator1
    Sprite Rotation Mode=::NewEnumerator2
    Sprite Size Mode=::NewEnumerator1
    Uniform Sprite Size Max=225
    Uniform Sprite Size Min=145
    UsePositionOffset=false
    Write Material Random=true
   )
   &SampleParticlesFromOtherEmitter()
   /SetVariables(Particles.RandomFloat=&RandomRangeFloat(Minimum=0.01, Maximum=1)) {isSetParameters=true}
   /SetVariables(Particles.InitSpriteRotRate=&UniformRangedFloat(Minimum=-3, Maximum=3)) {isSetParameters=true}
   &AddVelocity(Velocity=&RandomRangeVector(Minimum=<0 0 150>, Maximum=<0 0 800>))
   &ApplyInitialForces()
  ]
  particleUpdateScript [
   &ParticleState()
   /SetVariables(Particles.EmissiveParamBinding:NiagaraFloat=&FloatFromCurve(FloatCurve=DataInterface() {propertyValues {Curve {keys [
         {interpMode=RCIM_Cubic, tangentMode=RCTM_User, time=0, value=1}
         {arriveTangent=-2.0861, interpMode=RCIM_Cubic, leaveTangent=-2.0861, tangentMode=RCTM_User, time=0.2, value=0.25}
         <at 0.4 0>
         {interpMode=RCIM_Cubic, tangentMode=RCTM_User, time=1, value=0}
        ]}}})) {isSetParameters=true}
   &SpriteRotationRate(Rotation Rate=&Multiply_Float(
     A=$Particles.InitSpriteRotRate
     B=&FloatFromCurve(
      FloatCurve=DataInterface() {propertyValues {Curve {keys [
          {arriveTangent=-1.9214, interpMode=RCIM_Cubic, leaveTangent=-1.9214, tangentMode=RCTM_User, time=-0.0046, value=1.0113}
          {arriveTangent=-0.2496, interpMode=RCIM_Cubic, leaveTangent=-0.2496, tangentMode=RCTM_User, time=0.9896, value=0.16}
         ]}}}
      CurveIndex=$Emitter.Age
      Scale Curve=150
     )
    ))
   &ScaleColor(
    Scale Alpha=&FloatFromCurve(
     FloatCurve=DataInterface() {propertyValues {Curve {keys [{interpMode=RCIM_Cubic, tangentMode=RCTM_User, time=0, value=1}, {interpMode=RCIM_Cubic, tangentMode=RCTM_User, time=1, value=0}]}}}
     Scale Curve=2.5
    )
    Scale RGB=&VectorFromFloat(Value=&FloatFromCurve(FloatCurve=DataInterface() {propertyValues {Curve {keys [<at 0.1505 1>,<at 1.1 0.75>,<at 5 2>]}}}))
    ScaleA=true
    ScaleRGB=false
   )
   &Drag(
    Drag=&FloatFromCurve(FloatCurve=DataInterface() {propertyValues {Curve {keys [<at 0 1>,<at 1 0>]}}}, Scale Curve=4)
    Use Linear Drag=true
    Use Rotational Drag=false
   )
   &GravityForce(Gravity=&Lerp_Vector(A=<0 0 100>, B=<0 0 -100>, Alpha=$Particles.MaterialRandom))
   /Niagara/Modules/Update/Forces/CurlNoiseForce(Noise Frequency=0.01, Noise Strength=500, Pan Noise=true, Pan Noise Field=<0 0 1>)
   &SolveForcesAndVelocity(Clamp Velocity=false, Limit Acceleration=false)
   &ScaleSpriteSize(Scale Factor=&Vector2DFromFloat(Value=&FloatFromCurve(
      FloatCurve=DataInterface() {propertyValues {Curve {keys [
          {arriveTangent=1.503, interpMode=RCIM_Cubic, leaveTangent=1.503, tangentMode=RCTM_User, time=0, value=0.25}
          {interpMode=RCIM_Cubic, tangentMode=RCTM_User, time=1, value=1}
         ]}}}
      Scale Curve=3
     ))) {module=ScaleSpriteSize001}
   &DynamicMaterialParameters(
    Index 0 Param 1=&Add_Float(
     A=$Particles.RandomFloat
     B=&FloatFromCurve(
      FloatCurve=DataInterface() {propertyValues {Curve {keys [
          {arriveTangent=2.7533, interpMode=RCIM_Cubic, leaveTangent=1.8958, tangentMode=RCTM_User, time=0, value=1}
          {arriveTangent=0.8292, interpMode=RCIM_Cubic, leaveTangent=0.8292, tangentMode=RCTM_User, time=0.321, value=1.3453}
          {arriveTangent=0.0529, interpMode=RCIM_Cubic, leaveTangent=0.2021, tangentMode=RCTM_User, time=1, value=1.5}
         ]}}}
      CurveIndex=&Add_Float(A=&Multiply_Float(A=$Particles.Initial.MaterialRandom, B=0.5), B=$Particles.NormalizedAge)
     )
    )
    Index 0 Param 2=&FloatFromCurve(FloatCurve=DataInterface() {propertyValues {Curve {keys [
         {arriveTangent=0.0453, interpMode=RCIM_Cubic, tangentMode=RCTM_Break, time=0, value=-1}
         {arriveTangent=0.0604, interpMode=RCIM_Cubic, leaveTangent=0.0604, tangentMode=RCTM_User, time=1, value=-1}
        ]}}})
    Index 0 Param 4=&FloatFromCurve(FloatCurve=DataInterface() {propertyValues {Curve {keys [{interpMode=RCIM_Cubic, leaveTangent=-1.5927, tangentMode=RCTM_Break, time=0, value=1}, <at 1 0>]}}})
    Index 1 Param 1=&FloatFromCurve(
     FloatCurve=DataInterface() {propertyValues {Curve {keys [
         {interpMode=RCIM_Cubic, leaveTangent=4.7088, tangentMode=RCTM_Break, time=0, value=0}
         {arriveTangent=4.2257, interpMode=RCIM_Cubic, leaveTangent=4.2257, tangentMode=RCTM_User, time=0.25, value=1}
        ]}}}
     Scale Curve=&RandomRangeFloat(Minimum=1.5, Maximum=1.5)
    )
    Index 1 Param 2=&FloatFromCurve(
     FloatCurve=DataInterface() {propertyValues {Curve {keys [<at 0 0.2>, {leaveTangent=0.9066, time=1, value=2}]}}}
     Scale Curve=&RandomRangeFloat(Minimum=0.5, Maximum=0.6)
    )
    Index 1 Param 3=&FloatFromCurve(FloatCurve=DataInterface() {propertyValues {Curve {keys [
         {interpMode=RCIM_Cubic, leaveTangent=6.3917, tangentMode=RCTM_Break, time=0, value=0.2}
         {arriveTangent=4.8433, interpMode=RCIM_Cubic, leaveTangent=19.0929, tangentMode=RCTM_User, time=0.25, value=1}
        ]}}})
    Index 1 Param 4=&FloatFromCurve(
     FloatCurve=DataInterface() {propertyValues {Curve {keys [{arriveTangent=-0.3224, interpMode=RCIM_Cubic, leaveTangent=-0.3224, tangentMode=RCTM_User, time=0, value=1}, <at 1 0.9>]}}}
     Scale Curve=&RandomRangeFloat(Minimum=1.1, Maximum=1.4)
    )
    Index 2 Param 1=&FloatFromCurve(FloatCurve=DataInterface() {propertyValues {Curve {keys [
         {interpMode=RCIM_Cubic, leaveTangent=4.2485, tangentMode=RCTM_Break, time=0, value=0}
         {arriveTangent=1.0548, interpMode=RCIM_Cubic, leaveTangent=1.0548, tangentMode=RCTM_User, time=0.17, value=0.5715}
         <at 1 1>
        ]}}})
    Index 2 Param 2=&FloatFromCurve(
     FloatCurve=DataInterface() {propertyValues {Curve {keys [
         {arriveTangent=3.069, interpMode=RCIM_Cubic, leaveTangent=3.069, tangentMode=RCTM_User, time=0, value=0.2}
         {arriveTangent=0.0259, interpMode=RCIM_Cubic, leaveTangent=0.0259, tangentMode=RCTM_User, time=1, value=1.4}
        ]}}}
     Scale Curve=&RandomRangeFloat(Minimum=0.2, Maximum=1.25)
    )
    Index 2 Param 3=&FloatFromCurve(FloatCurve=DataInterface() {propertyValues {Curve {keys [
         {interpMode=RCIM_Cubic, leaveTangent=2.3064, tangentMode=RCTM_Break, time=0, value=1}
         {interpMode=RCIM_Cubic, leaveTangent=11.1233, tangentMode=RCTM_Break, time=1, value=2}
        ]}}})
    Index 2 Param 4=&FloatFromCurve(FloatCurve=DataInterface() {propertyValues {Curve {keys [
         {arriveTangent=-0.6867, interpMode=RCIM_Cubic, leaveTangent=-0.6867, tangentMode=RCTM_User, time=0, value=1}
         {interpMode=RCIM_Cubic, leaveTangent=-1.2956, tangentMode=RCTM_Break, time=1, value=0.5}
        ]}}})
    Module.Param0WriteEnabled=true
    Module.Param0WriteEnabled1=true
    Module.Param0WriteEnabled2=true
    Module.Param1WriteEnabled=false
    Module.Param1WriteEnabled1=true
    Module.Param1WriteEnabled2=true
    Module.Param2WriteEnabled=true
    Module.Param2WriteEnabled1=true
    Module.Param2WriteEnabled2=true
    Module.Param3WriteEnabled=true
    Module.Param3WriteEnabled1=true
    Module.Param3WriteEnabled2=true
    Write Parameter Index 1=true
    Write Parameter Index 2=true
   )
  ]
  rendererClasses [&NiagaraSpriteRendererProperties]
  renderers [{rendererClass=&NiagaraSpriteRendererProperties}]
 }
 Trail_Sparks {
  emitterUpdateScript [
   &EmitterState()
   &SpawnParticlesFromOtherEmitter(
    Attribute Reader=DataInterface() {propertyValues {EmitterBinding {bindingMode=Other, emitterName=Source}}}
    Spawn Rate=75
    Spawn Rate Per Particle Cap=512
    Use Maximum Age=false
    Use Minimum Age=false
   )
  ]
  particleSpawnScript [
   &InitializeParticle(
    Color=$System.SystemColor
    Lifetime Max=0.75
    Lifetime Min=0.5
    Lifetime Mode=::NewEnumerator1
    Sprite Size=&RandomRangeVector2D(Minimum=<30 125>, Maximum=<40 150>)
    Sprite Size Mode=::NewEnumerator4
    UsePositionOffset=false
    Write Material Random=true
   )
   &SampleParticlesFromOtherEmitter()
   &AddVelocity(
    Module.Use Falloff Distance=false
    Override Seed=false
    Velocity Mode=::NewEnumerator1
    Velocity Speed=&RandomRangeFloat(Minimum=-900, Maximum=-1000)
   )
   &ShapeLocation(MaskSpawnGroup=false, Override Seed=false, Sphere Radius=50, UseOwnerScale=true)
  ]
  particleUpdateScript [
   &ParticleState()
   &ScaleColor(
    Scale Alpha=&FloatFromCurve(FloatCurve=DataInterface() {propertyValues {Curve {keys [
         {arriveTangent=0.0019, interpMode=RCIM_Cubic, leaveTangent=0.0019, tangentMode=RCTM_User, time=0, value=1}
         {interpMode=RCIM_Cubic, tangentMode=RCTM_User, time=0.5, value=1}
         {arriveTangent=0.0018, interpMode=RCIM_Cubic, leaveTangent=0.0018, tangentMode=RCTM_User, time=1.0011, value=0}
        ]}}})
    Scale RGB=&VectorFromFloat(Value=3)
    ScaleA=true
    ScaleRGB=true
   ) {module=ScaleColor002}
   &Drag(Use Linear Drag=true, Use Rotational Drag=true)
   /Niagara/Modules/Update/Forces/V2/CurlNoiseForce(
    Noise Frequency=100
    Noise Strength=&Multiply_Float(
     A=1500
     B=&FloatFromCurve(FloatCurve=DataInterface() {propertyValues {Curve {keys [
          {arriveTangent=4.0408, interpMode=RCIM_Cubic, leaveTangent=4.0408, tangentMode=RCTM_User, time=0, value=0.1}
          {arriveTangent=0.4348, interpMode=RCIM_Cubic, leaveTangent=0.4348, tangentMode=RCTM_User, time=0.1, value=0.9}
          <at 1 1>
         ]}}})
    )
    Pan Noise=false
   )
   &SolveForcesAndVelocity(Clamp Velocity=false, Limit Acceleration=false)
   &ScaleSpriteSize(
    Uniform Curve Scale=0.75
    Uniform Curve Sprite Scale=DataInterface() {propertyValues {Curve {keys [
        {arriveTangent=16.1993, interpMode=RCIM_Cubic, leaveTangent=16.1993, tangentMode=RCTM_User, time=0, value=0}
        <at 0.1408 0.9847 RCIM_Cubic>
        {arriveTangent=-0.0405, interpMode=RCIM_Cubic, leaveTangent=-0.0405, tangentMode=RCTM_User, time=1, value=0}
       ]}}}
   )
  ]
  rendererClasses [&NiagaraSpriteRendererProperties]
  renderers [{rendererClass=&NiagaraSpriteRendererProperties}]
 }
}
summary {
 systemName=NS_Sparks_Firework_Sphere_Double
 emitters [
  {emitterName=Light_Glow, rendererClasses [&NiagaraSpriteRendererProperties]}
  {emitterName=Source, rendererClasses [&NiagaraSpriteRendererProperties]}
  {emitterName=Ribbon_Smoke, rendererClasses [&NiagaraRibbonRendererProperties]}
  {emitterName=Source_SparkTrail001, simTarget=GPUComputeSim, rendererClasses [&NiagaraSpriteRendererProperties]}
  {emitterName=Trail_Sparks, rendererClasses [&NiagaraSpriteRendererProperties]}
  {emitterName=Trail_Smoke, rendererClasses [&NiagaraSpriteRendererProperties]}
  {emitterName=FIrework_Blast_Combined, rendererClasses [&NiagaraMeshRendererProperties]}
  {emitterName=Mesh_Ribbons_Combined, rendererClasses [&NiagaraMeshRendererProperties]}
 ]
 userVariables [User.HeightMult {type=NiagaraFloat, defaultValue=1, description=""}]
}
