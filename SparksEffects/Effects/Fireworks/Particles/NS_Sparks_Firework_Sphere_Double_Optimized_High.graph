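# Niagara system for a double sphere firework: a shell launches upward with smoke and ember trails, then bursts at its death position.
# MainSource is the shell, writing its death into an array, and the trail emitters spawn from its particles.
# Optimized blast, glow and ribbon-mesh emitters spawn from system-level DeathSpawn and DeathPosition values, while their timed, particle-reader counterparts are disabled.
# The burst animates meshes and sprites through curve-driven dynamic material parameters.
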NiagaraSystem /SparksEffects/Effects/Fireworks/Particles/NS_Sparks_Firework_Sphere_Double_Optimized_High
dependencies {
 usedDataInterfaces [
  &NiagaraDataInterfaceParticleRead
  &NiagaraDataInterfaceCurve
  &NiagaraDataInterfaceArrayPosition
  &NiagaraDataInterfaceArrayColor
 ]
 usedDynamicInputs [
  &MakeVector
  &Sine
  &SelectLinearColorFromArray
 ]
 usedModules [
  &self:SystemSpawnScript.SystemScriptSource.SystemScriptGraph.NiagaraNodeAssignment_1.SetVariables_#1
  &SystemState
  &self:Sys_ReadDeath
  &self:SystemSpawnScript.SystemScriptSource.SystemScriptGraph.NiagaraNodeAssignment_0.SetVariables_#2
 ]
}
emitters {
 FIrework_Blast {
  bEnabled=false
  emitterUpdateScript [
   &EmitterState()
   &SpawnBurst_Instantaneous(
    Spawn Time=&Subtract_Float(A=$System.LifetimeRand, B=0.25)
    Use Spawn Probability=false
    UseLoopCountLimit=false
   )
  ]
  particleSpawnScript [
   &InitializeParticle(
    Color=$System.SystemColor
    Lifetime=4
    Mesh Scale Mode=::NewEnumerator3
    Mesh Uniform Scale=15
    UsePositionOffset=false
    Write Material Random=true
   )
   &self:FireworkPosition(Particle Attribute Reader=DataInterface() {propertyValues {EmitterBinding {bindingMode=Other, emitterName=MainSource}}}) {module=FireworkPosition}
  ]
  particleUpdateScript [
   &ParticleState()
   &ScaleMeshSize(Scale Factor=&VectorFromFloat(Value=$System.BurstMeshSize))
   /SetVariables(Particles.Scale Mesh Size Blast Small:Vector3f=&Multiply_Vector(
     A=&VectorFromFloat(Value=10)
     B=&VectorFromFloat(Value=$System.BurstMeshSize)
    )) {isSetParameters=true, module=SetVariables_#3}
   &ScaleColor(
    Scale Alpha=&FloatFromCurve(FloatCurve=DataInterface() {propertyValues {Curve {keys [
         {arriveTangent=0.0019, interpMode=RCIM_Cubic, leaveTangent=0.0019, tangentMode=RCTM_User, time=0, value=1}
         {interpMode=RCIM_Cubic, tangentMode=RCTM_User, time=0.5, value=1}
         {arriveTangent=0.0018, interpMode=RCIM_Cubic, leaveTangent=0.0018, tangentMode=RCTM_User, time=0.75, value=0}
        ]}}})
    Scale RGB=&VectorFromFloat(Value=3)
    ScaleA=true
    ScaleRGB=true
   )
   /SetVariables(Particles.Blast Small Color:LinearColor=&MultiplyLinearColor(
     Linear Color=&HueShiftLinearColor(Hue=0.2, Saturation=1, Value=1, Linear Color=$System.SystemColor)
     Scale Factor=&MakeVector4(
      X=2
      Y=2
      Z=2
      W=&FloatFromCurve(FloatCurve=DataInterface() {propertyValues {Curve {keys [
           {arriveTangent=0.0019, interpMode=RCIM_Cubic, leaveTangent=0.0019, tangentMode=RCTM_User, time=0, value=1}
           {interpMode=RCIM_Cubic, tangentMode=RCTM_User, time=0.2, value=1}
           {arriveTangent=0.0018, interpMode=RCIM_Cubic, leaveTangent=0.0018, tangentMode=RCTM_User, time=1.0011, value=0}
          ]}}})
     )
    )) {isSetParameters=true, module=SetVariables_#4}
   &DynamicMaterialParameters(
    Index 0 Param 1=&FloatFromCurve(
     FloatCurve=DataInterface() {propertyValues {Curve {keys [
         {arriveTangent=-0.0112, interpMode=RCIM_Cubic, leaveTangent=-0.0112, tangentMode=RCTM_User, time=0, value=0.4}
         {arriveTangent=-0.7264, interpMode=RCIM_Cubic, leaveTangent=-0.7264, tangentMode=RCTM_User, time=0.1, value=0.3}
         {arriveTangent=-0.1797, interpMode=RCIM_Cubic, leaveTangent=-0.1797, tangentMode=RCTM_User, time=0.4, value=0.225}
         {arriveTangent=-0.1859, interpMode=RCIM_Cubic, leaveTangent=-0.1859, tangentMode=RCTM_User, time=0.6, value=0.175}
         {arriveTangent=-3.4317, interpMode=RCIM_Cubic, leaveTangent=-3.4317, tangentMode=RCTM_User, time=0.75, value=0}
        ]}}}
     Scale Curve=300
    )
    Index 0 Param 2=&FloatFromCurve(FloatCurve=DataInterface() {propertyValues {Curve {keys [
         {arriveTangent=14.9629, interpMode=RCIM_Cubic, leaveTangent=14.9629, tangentMode=RCTM_User, time=0, value=-0.7}
         {arriveTangent=3.0529, interpMode=RCIM_Cubic, leaveTangent=3.0529, tangentMode=RCTM_User, time=0.16, value=0.5}
         {interpMode=RCIM_Cubic, tangentMode=RCTM_User, time=0.75, value=1}
        ]}}})
    Index 0 Param 3=&FloatFromCurve(
     FloatCurve=DataInterface() {propertyValues {Curve {keys [
         {interpMode=RCIM_Cubic, tangentMode=RCTM_User, time=0, value=0}
         {arriveTangent=-856.4041, interpMode=RCIM_Cubic, leaveTangent=-856.4041, tangentMode=RCTM_User, time=0.75, value=-500}
        ]}}}
     Scale Curve=3
    )
    Index 0 Param 4=&FloatFromCurve(
     FloatCurve=DataInterface() {propertyValues {Curve {keys [
         {arriveTangent=0.0932, interpMode=RCIM_Cubic, leaveTangent=0.0932, tangentMode=RCTM_User, time=0, value=0}
         {arriveTangent=1.4953, interpMode=RCIM_Cubic, leaveTangent=1.4953, tangentMode=RCTM_User, time=0.75, value=1}
        ]}}}
     Scale Curve=600
    )
    Index 1 Param 1=-0.1
    Module.Param0WriteEnabled=true
    Module.Param0WriteEnabled1=true
    Module.Param1WriteEnabled=true
    Module.Param1WriteEnabled1=true
    Module.Param2WriteEnabled=true
    Module.Param2WriteEnabled1=true
    Module.Param3WriteEnabled=true
    Module.Param3WriteEnabled1=true
    Write Parameter Index 1=true
   )
   /SetVariables(Particles.Blast Small Dynamic Param 0:Vector4f=&MakeVector4(
     X=&FloatFromCurve(
      FloatCurve=DataInterface() {propertyValues {Curve {keys [
          {arriveTangent=-0.0112, interpMode=RCIM_Cubic, leaveTangent=-0.0112, tangentMode=RCTM_User, time=0, value=0.4}
          {arriveTangent=-0.7264, interpMode=RCIM_Cubic, leaveTangent=-0.7264, tangentMode=RCTM_User, time=0.1, value=0.3}
          {arriveTangent=-0.1797, interpMode=RCIM_Cubic, leaveTangent=-0.1797, tangentMode=RCTM_User, time=0.4326, value=0.2242}
          {arriveTangent=-0.1859, interpMode=RCIM_Cubic, leaveTangent=-0.1859, tangentMode=RCTM_User, time=0.8879, value=0.1764}
          {arriveTangent=-3.4317, interpMode=RCIM_Cubic, leaveTangent=-3.4317, tangentMode=RCTM_User, time=1, value=0}
         ]}}}
      Scale Curve=300
     )
     Y=&FloatFromCurve(FloatCurve=DataInterface() {propertyValues {Curve {keys [
          {arriveTangent=10.0245, interpMode=RCIM_Cubic, leaveTangent=10.0245, tangentMode=RCTM_User, time=0, value=-1}
          {arriveTangent=1.4411, interpMode=RCIM_Cubic, leaveTangent=1.4411, tangentMode=RCTM_User, time=0.3, value=0.375}
          {interpMode=RCIM_Cubic, tangentMode=RCTM_User, time=1, value=0.75}
         ]}}})
     Z=&FloatFromCurve(
      FloatCurve=DataInterface() {propertyValues {Curve {keys [
          {interpMode=RCIM_Cubic, tangentMode=RCTM_User, time=0, value=0}
          {arriveTangent=-856.4041, interpMode=RCIM_Cubic, leaveTangent=-856.4041, tangentMode=RCTM_User, time=1, value=-500}
         ]}}}
      Scale Curve=2
     )
     W=&FloatFromCurve(
      FloatCurve=DataInterface() {propertyValues {Curve {keys [
          {arriveTangent=0.0932, interpMode=RCIM_Cubic, leaveTangent=0.0932, tangentMode=RCTM_User, time=0, value=0}
          {arriveTangent=1.4953, interpMode=RCIM_Cubic, leaveTangent=1.4953, tangentMode=RCTM_User, time=1, value=1}
         ]}}}
      Scale Curve=300
     )
    )) {isSetParameters=true, module=SetVariables_#5}
   /SetVariables() {isSetParameters=true, module=SetVariables_#6}
  ]
  rendererClasses [&NiagaraMeshRendererProperties]
  renderers [
   {rendererClass=&NiagaraMeshRendererProperties, rendererIndex=0}
   {rendererClass=&NiagaraMeshRendererProperties, rendererIndex=1}
  ]
 }
 FIrework_Blast_Optimized {
  emitterUpdateScript [
   &EmitterState()
   &SpawnBurst_Instantaneous(
    Spawn Count=&MakeCustomIntFromBool(Bool=$Output.Read_Death.ShouldSpawn, True INT=1, False INT=0)
    Use Spawn Probability=false
    UseLoopCountLimit=false
   ) {enabled=false}
   &SpawnPerFrame(Spawn:NiagaraBool=$System.DeathSpawn, Use Spawn Probability=false)
  ]
  particleSpawnScript [&InitializeParticle(
    Color=$System.SystemColor
    Lifetime=4
    Mesh Scale Mode=::NewEnumerator3
    Mesh Uniform Scale=15
    Position=$System.DeathPosition
    Position Mode=::NewEnumerator1
    UsePositionOffset=false
    Write Material Random=true
   )]
  particleUpdateScript [
   &ParticleState()
   &ScaleMeshSize(Scale Factor=&VectorFromFloat(Value=$System.BurstMeshSize))
   /SetVariables(Particles.Scale Mesh Size Blast Small:Vector3f=&Multiply_Vector(
     A=&VectorFromFloat(Value=10)
     B=&VectorFromFloat(Value=$System.BurstMeshSize)
    )) {isSetParameters=true, module=SetVariables_#3}
   &ScaleColor(
    Scale Alpha=&FloatFromCurve(FloatCurve=DataInterface() {propertyValues {Curve {keys [
         {arriveTangent=0.0019, interpMode=RCIM_Cubic, leaveTangent=0.0019, tangentMode=RCTM_User, time=0, value=1}
         {interpMode=RCIM_Cubic, tangentMode=RCTM_User, time=0.5, value=1}
         {arriveTangent=0.0018, interpMode=RCIM_Cubic, leaveTangent=0.0018, tangentMode=RCTM_User, time=0.75, value=0}
        ]}}})
    Scale RGB=&VectorFromFloat(Value=3)
    ScaleA=true
    ScaleRGB=true
   )
   /SetVariables(Particles.Blast Small Color:LinearColor=&MultiplyLinearColor(
     Linear Color=&HueShiftLinearColor(Hue=0.2, Saturation=1, Value=1, Linear Color=$System.SystemColor)
     Scale Factor=&MakeVector4(
      X=2
      Y=2
      Z=2
      W=&FloatFromCurve(FloatCurve=DataInterface() {propertyValues {Curve {keys [
           {arriveTangent=0.0019, interpMode=RCIM_Cubic, leaveTangent=0.0019, tangentMode=RCTM_User, time=0, value=1}
           {interpMode=RCIM_Cubic, tangentMode=RCTM_User, time=0.2, value=1}
           {arriveTangent=0.0018, interpMode=RCIM_Cubic, leaveTangent=0.0018, tangentMode=RCTM_User, time=1.0011, value=0}
          ]}}})
     )
    )) {isSetParameters=true, module=SetVariables_#4}
   &DynamicMaterialParameters(
    Index 0 Param 1=&FloatFromCurve(
     FloatCurve=DataInterface() {propertyValues {Curve {keys [
         {arriveTangent=-0.0112, interpMode=RCIM_Cubic, leaveTangent=-0.0112, tangentMode=RCTM_User, time=0, value=0.4}
         {arriveTangent=-0.7264, interpMode=RCIM_Cubic, leaveTangent=-0.7264, tangentMode=RCTM_User, time=0.1, value=0.3}
         {arriveTangent=-0.1797, interpMode=RCIM_Cubic, leaveTangent=-0.1797, tangentMode=RCTM_User, time=0.4, value=0.225}
         {arriveTangent=-0.1859, interpMode=RCIM_Cubic, leaveTangent=-0.1859, tangentMode=RCTM_User, time=0.6, value=0.175}
         {arriveTangent=-3.4317, interpMode=RCIM_Cubic, leaveTangent=-3.4317, tangentMode=RCTM_User, time=0.75, value=0}
        ]}}}
     Scale Curve=300
    )
    Index 0 Param 2=&FloatFromCurve(FloatCurve=DataInterface() {propertyValues {Curve {keys [
         {arriveTangent=14.9629, interpMode=RCIM_Cubic, leaveTangent=14.9629, tangentMode=RCTM_User, time=0, value=-0.7}
         {arriveTangent=3.0529, interpMode=RCIM_Cubic, leaveTangent=3.0529, tangentMode=RCTM_User, time=0.16, value=0.5}
         {interpMode=RCIM_Cubic, tangentMode=RCTM_User, time=0.75, value=1}
        ]}}})
    Index 0 Param 3=&FloatFromCurve(
     FloatCurve=DataInterface() {propertyValues {Curve {keys [
         {interpMode=RCIM_Cubic, tangentMode=RCTM_User, time=0, value=0}
         {arriveTangent=-856.4041, interpMode=RCIM_Cubic, leaveTangent=-856.4041, tangentMode=RCTM_User, time=0.75, value=-500}
        ]}}}
     Scale Curve=3
    )
    Index 0 Param 4=&FloatFromCurve(
     FloatCurve=DataInterface() {propertyValues {Curve {keys [
         {arriveTangent=0.0932, interpMode=RCIM_Cubic, leaveTangent=0.0932, tangentMode=RCTM_User, time=0, value=0}
         {arriveTangent=1.4953, interpMode=RCIM_Cubic, leaveTangent=1.4953, tangentMode=RCTM_User, time=0.75, value=1}
        ]}}}
     Scale Curve=600
    )
    Index 1 Param 1=-0.1
    Module.Param0WriteEnabled=true
    Module.Param0WriteEnabled1=true
    Module.Param1WriteEnabled=true
    Module.Param1WriteEnabled1=true
    Module.Param2WriteEnabled=true
    Module.Param2WriteEnabled1=true
    Module.Param3WriteEnabled=true
    Module.Param3WriteEnabled1=true
    Write Parameter Index 1=true
   )
   /SetVariables(Particles.Blast Small Dynamic Param 0:Vector4f=&MakeVector4(
     X=&FloatFromCurve(
      FloatCurve=DataInterface() {propertyValues {Curve {keys [
          {arriveTangent=-0.0112, interpMode=RCIM_Cubic, leaveTangent=-0.0112, tangentMode=RCTM_User, time=0, value=0.4}
          {arriveTangent=-0.7264, interpMode=RCIM_Cubic, leaveTangent=-0.7264, tangentMode=RCTM_User, time=0.1, value=0.3}
          {arriveTangent=-0.1797, interpMode=RCIM_Cubic, leaveTangent=-0.1797, tangentMode=RCTM_User, time=0.4326, value=0.2242}
          {arriveTangent=-0.1859, interpMode=RCIM_Cubic, leaveTangent=-0.1859, tangentMode=RCTM_User, time=0.8879, value=0.1764}
          {arriveTangent=-3.4317, interpMode=RCIM_Cubic, leaveTangent=-3.4317, tangentMode=RCTM_User, time=1, value=0}
         ]}}}
      Scale Curve=300
     )
     Y=&FloatFromCurve(FloatCurve=DataInterface() {propertyValues {Curve {keys [
          {arriveTangent=10.0245, interpMode=RCIM_Cubic, leaveTangent=10.0245, tangentMode=RCTM_User, time=0, value=-1}
          {arriveTangent=1.4411, interpMode=RCIM_Cubic, leaveTangent=1.4411, tangentMode=RCTM_User, time=0.3, value=0.375}
          {interpMode=RCIM_Cubic, tangentMode=RCTM_User, time=1, value=0.75}
         ]}}})
     Z=&FloatFromCurve(
      FloatCurve=DataInterface() {propertyValues {Curve {keys [
          {interpMode=RCIM_Cubic, tangentMode=RCTM_User, time=0, value=0}
          {arriveTangent=-856.4041, interpMode=RCIM_Cubic, leaveTangent=-856.4041, tangentMode=RCTM_User, time=1, value=-500}
         ]}}}
      Scale Curve=2
     )
     W=&FloatFromCurve(
      FloatCurve=DataInterface() {propertyValues {Curve {keys [
          {arriveTangent=0.0932, interpMode=RCIM_Cubic, leaveTangent=0.0932, tangentMode=RCTM_User, time=0, value=0}
          {arriveTangent=1.4953, interpMode=RCIM_Cubic, leaveTangent=1.4953, tangentMode=RCTM_User, time=1, value=1}
         ]}}}
      Scale Curve=300
     )
    )) {isSetParameters=true, module=SetVariables_#5}
   /SetVariables() {isSetParameters=true, module=SetVariables_#6}
  ]
  rendererClasses [&NiagaraMeshRendererProperties]
  renderers [
   {rendererClass=&NiagaraMeshRendererProperties, rendererIndex=0}
   {rendererClass=&NiagaraMeshRendererProperties, rendererIndex=1}
  ]
 }
 Light_Glow {
  bEnabled=false
  emitterUpdateScript [
   &EmitterState()
   &SpawnBurst_Instantaneous(
    Spawn Time=&Subtract_Float(A=$System.LifetimeRand, B=0.25)
    Use Spawn Probability=false
    UseLoopCountLimit=false
   )
  ]
  particleSpawnScript [
   &InitializeParticle(
    Color=$System.SystemColor
    Lifetime=3
    Sprite Size Mode=::NewEnumerator3
    Uniform Sprite Size=20
    UsePositionOffset=false
    Write Material Random=true
   )
   &DynamicMaterialParameters(
    Index 0 Param 1=2.75
    Module.Index0Param0Write=true
    Module.Index0Param1Write=true
    Module.Index0Param2Write=true
    Module.Index0Param3Write=true
   )
   &self:FireworkPosition(Particle Attribute Reader=DataInterface() {propertyValues {EmitterBinding {bindingMode=Other, emitterName=MainSource}}}) {module=FireworkPosition}
  ]
  particleUpdateScript [
   &ParticleState()
   &ScaleSpriteSize(
    Scale Sprite Size Mode=::NewEnumerator0
    Uniform Scale Factor=&Multiply_Float(A=$System.BurstMeshSize, B=1.15)
   ) {module=ScaleSpriteSize001}
   &ScaleSpriteSize(
    Uniform Curve Scale=450
    Uniform Curve Sprite Scale=DataInterface() {propertyValues {Curve {keys [
        {arriveTangent=2.1659, interpMode=RCIM_Cubic, leaveTangent=2.1659, tangentMode=RCTM_User, time=0, value=0.3}
        {interpMode=RCIM_Cubic, tangentMode=RCTM_User, time=1, value=1}
       ]}}}
   )
   &ScaleColor(
    Scale Alpha=&FloatFromCurve(FloatCurve=DataInterface() {propertyValues {Curve {keys [
         {interpMode=RCIM_Cubic, tangentMode=RCTM_User, time=0.2, value=1}
         {arriveTangent=-1.7914, interpMode=RCIM_Cubic, leaveTangent=-1.7914, tangentMode=RCTM_User, time=1, value=0}
        ]}}})
    Scale RGB=&VectorFromFloat(Value=1.5)
    ScaleA=true
    ScaleRGB=true
   )
  ]
  rendererClasses [&NiagaraSpriteRendererProperties]
  renderers [{rendererClass=&NiagaraSpriteRendererProperties}]
 }
 Light_Glow_Optimized {
  emitterUpdateScript [
   &EmitterState()
   &SpawnBurst_Instantaneous(
    Spawn Count=&MakeCustomIntFromBool(Bool=$Output.Read_Death.ShouldSpawn, True INT=1, False INT=0)
    Use Spawn Probability=false
    UseLoopCountLimit=false
   ) {enabled=false}
   &SpawnPerFrame(Spawn:NiagaraBool=$System.DeathSpawn, Use Spawn Probability=false)
  ]
  particleSpawnScript [
   &InitializeParticle(
    Color=$System.SystemColor
    Lifetime=3
    Position=$System.DeathPosition
    Position Mode=::NewEnumerator1
    Sprite Size Mode=::NewEnumerator3
    Uniform Sprite Size=20
    UsePositionOffset=false
    Write Material Random=true
   )
   &DynamicMaterialParameters(
    Index 0 Param 1=2.75
    Module.Index0Param0Write=true
    Module.Index0Param1Write=true
    Module.Index0Param2Write=true
    Module.Index0Param3Write=true
   )
  ]
  particleUpdateScript [
   &ParticleState()
   &ScaleSpriteSize(
    Scale Sprite Size Mode=::NewEnumerator0
    Uniform Scale Factor=&Multiply_Float(A=$System.BurstMeshSize, B=1.15)
   ) {module=ScaleSpriteSize001}
   &ScaleSpriteSize(
    Uniform Curve Scale=450
    Uniform Curve Sprite Scale=DataInterface() {propertyValues {Curve {keys [
        {arriveTangent=2.1659, interpMode=RCIM_Cubic, leaveTangent=2.1659, tangentMode=RCTM_User, time=0, value=0.3}
        {interpMode=RCIM_Cubic, tangentMode=RCTM_User, time=1, value=1}
       ]}}}
   )
   &ScaleColor(
    Scale Alpha=&FloatFromCurve(FloatCurve=DataInterface() {propertyValues {Curve {keys [
         {interpMode=RCIM_Cubic, tangentMode=RCTM_User, time=0.2, value=1}
         {arriveTangent=-1.7914, interpMode=RCIM_Cubic, leaveTangent=-1.7914, tangentMode=RCTM_User, time=1, value=0}
        ]}}})
    Scale RGB=&VectorFromFloat(Value=1.5)
    ScaleA=true
    ScaleRGB=true
   )
  ]
  rendererClasses [&NiagaraSpriteRendererProperties]
  renderers [{rendererClass=&NiagaraSpriteRendererProperties}]
 }
 MainSource {
  emitterUpdateScript [
   &EmitterState()
   &SpawnBurst_Instantaneous(Use Spawn Probability=false, UseLoopCountLimit=false)
  ]
  particleSpawnScript [
   &InitializeParticle(
    Color=$System.SystemColor
    Lifetime=$System.LifetimeRand
    Position Offset=<0 0 -850>
    Sprite Size=<2000 2000>
    Sprite Size Mode=::NewEnumerator4
    UsePositionOffset=true
    Write Material Random=true
   )
   &AddVelocity(Velocity=<0 0 20000>)
  ]
  particleUpdateScript [
   &ParticleState()
   &self:Write_Death(Position Array:NiagaraDataInterfaceArrayPosition=$System.Death Array) {module=Write_Death}
   &Drag(Use Linear Drag=true, Use Rotational Drag=true)
   &GravityForce()
   &SolveForcesAndVelocity(Clamp Velocity=false, Limit Acceleration=false)
   &DynamicMaterialParameters(
    Index 0 Param 1=0.8
    Index 0 Param 2=1
    Index 0 Param 3=1
    Index 1 Param 1=150
    Index 1 Param 4=1
    Module.Param0WriteEnabled=true
    Module.Param0WriteEnabled1=true
    Module.Param1WriteEnabled=true
    Module.Param1WriteEnabled1=true
    Module.Param2WriteEnabled=true
    Module.Param2WriteEnabled1=true
    Module.Param3WriteEnabled=false
    Module.Param3WriteEnabled1=true
    Write Parameter Index 1=true
   )
   &ScaleColor(
    Scale Alpha=&FloatFromCurve(FloatCurve=DataInterface() {propertyValues {Curve {keys [
         {arriveTangent=9.0725, interpMode=RCIM_Cubic, leaveTangent=9.0725, tangentMode=RCTM_User, time=0, value=0}
         {arriveTangent=-7.2695, interpMode=RCIM_Cubic, leaveTangent=-7.2695, tangentMode=RCTM_User, time=0.2, value=1.75}
         <at 0.8 1.8>
         {arriveTangent=-0.0405, interpMode=RCIM_Cubic, leaveTangent=-0.0405, tangentMode=RCTM_User, time=1, value=0}
        ]}}})
    ScaleA=true
    ScaleRGB=true
   ) {module=ScaleColor001}
   &ScaleSpriteSize(Uniform Curve Sprite Scale=DataInterface() {propertyValues {Curve {keys [
        {interpMode=RCIM_Cubic, tangentMode=RCTM_User, time=0, value=1}
        {arriveTangent=-0.0015, interpMode=RCIM_Cubic, leaveTangent=-0.0015, tangentMode=RCTM_User, time=1, value=0.5}
       ]}}}) {module=ScaleSpriteSize001}
  ]
  rendererClasses [&NiagaraSpriteRendererProperties]
  renderers [{rendererClass=&NiagaraSpriteRendererProperties}]
 }
 Mesh_Ribbons {
  bEnabled=false
  emitterUpdateScript [
   &EmitterState() {module=EmitterState001}
   &SpawnBurst_Instantaneous(
    Spawn Time=&Subtract_Float(A=$System.LifetimeRand, B=0.25)
    Use Spawn Probability=false
    UseLoopCountLimit=false
   )
  ]
  particleSpawnScript [
   &InitializeParticle(
    Color=$System.SystemColor
    Lifetime=&RandomRangeFloat(Minimum=2.5, Maximum=3)
    Mesh Scale Mode=::NewEnumerator3
    Sprite Size Mode=::NewEnumerator3
    Uniform Sprite Size=50
    UsePositionOffset=false
    Write Material Random=true
   )
   &InitialMeshOrientation(Use Rotation Vector=false)
   /SetVariables(Particles.Ribbon No Sparks Orientation:Quat4f=/Niagara/DynamicInputs/Rotation/RandomQuaternion()) {isSetParameters=true, module=SetVariables_#7}
   &self:FireworkPosition(Particle Attribute Reader=DataInterface() {propertyValues {EmitterBinding {bindingMode=Other, emitterName=MainSource}}}) {module=FireworkPosition}
  ]
  particleUpdateScript [
   &ParticleState()
   &ScaleMeshSize(Scale Factor=&VectorFromFloat(Value=$System.BurstMeshSize))
   &ScaleColor(
    Scale Alpha=&FloatFromCurve(FloatCurve=DataInterface() {propertyValues {Curve {keys [
         {arriveTangent=0.0543, interpMode=RCIM_Cubic, leaveTangent=0.0543, tangentMode=RCTM_User, time=0, value=0}
         {arriveTangent=0.0595, interpMode=RCIM_Cubic, leaveTangent=0.0595, tangentMode=RCTM_User, time=1, value=1}
        ]}}})
    Scale RGB=&VectorFromFloat(Value=&FloatFromCurve(FloatCurve=DataInterface() {propertyValues {Curve {keys [
          {arriveTangent=-33.3256, interpMode=RCIM_Cubic, leaveTangent=-33.3256, tangentMode=RCTM_User, time=0, value=5}
          {arriveTangent=-0.2459, interpMode=RCIM_Cubic, leaveTangent=-0.2459, tangentMode=RCTM_User, time=0.25, value=1}
         ]}}}))
    ScaleA=true
    ScaleRGB=true
   )
   &DynamicMaterialParameters(
    Index 0 Param 1=&FloatFromCurve(
     FloatCurve=DataInterface() {propertyValues {Curve {keys [
         {arriveTangent=4.4937, interpMode=RCIM_Cubic, leaveTangent=4.4937, tangentMode=RCTM_User, time=0, value=0}
         {arriveTangent=1.4506, interpMode=RCIM_Cubic, leaveTangent=1.4506, tangentMode=RCTM_User, time=0.2, value=0.625}
         {arriveTangent=0.0378, interpMode=RCIM_Cubic, leaveTangent=0.0378, tangentMode=RCTM_User, time=1, value=1}
        ]}}}
     Scale Curve=3
    )
    Index 0 Param 2=&FloatFromCurve(
     FloatCurve=DataInterface() {propertyValues {Curve {keys [
         {arriveTangent=-0.018, interpMode=RCIM_Cubic, leaveTangent=-0.018, tangentMode=RCTM_User, time=0, value=0}
         {arriveTangent=1.4018, interpMode=RCIM_Cubic, leaveTangent=1.4018, tangentMode=RCTM_User, time=1, value=1}
        ]}}}
     Scale Curve=2
    )
    Index 0 Param 3=-0.2
    Index 1 Param 1=1
    Index 1 Param 2=&FloatFromCurve(
     FloatCurve=DataInterface() {propertyValues {Curve {keys [
         {arriveTangent=-6.499, interpMode=RCIM_Cubic, leaveTangent=-6.499, tangentMode=RCTM_User, time=0, value=3}
         {interpMode=RCIM_Cubic, tangentMode=RCTM_User, time=1, value=1}
        ]}}}
     Scale Curve=3
    )
    Index 1 Param 3=0.5
    Index 1 Param 4=&FloatFromCurve(
     FloatCurve=DataInterface() {propertyValues {Curve {keys [
         {arriveTangent=0.0351, interpMode=RCIM_Cubic, leaveTangent=0.0351, tangentMode=RCTM_User, time=0, value=0}
         {interpMode=RCIM_Cubic, tangentMode=RCTM_User, time=0.7, value=1}
         {interpMode=RCIM_Cubic, tangentMode=RCTM_User, time=1, value=0}
        ]}}}
     Scale Curve=125
    )
    Index 2 Param 1=1
    Index 2 Param 2=&FloatFromCurve(
     FloatCurve=DataInterface() {propertyValues {Curve {keys [
         {arriveTangent=0.0495, interpMode=RCIM_Cubic, leaveTangent=0.0495, tangentMode=RCTM_User, time=0, value=0}
         {arriveTangent=1.1556, interpMode=RCIM_Cubic, leaveTangent=1.1556, tangentMode=RCTM_User, time=1, value=1}
        ]}}}
     Scale Curve=5
    )
    Index 2 Param 3=1
    Index 2 Param 4=0.35
    Module.Param0WriteEnabled=true
    Module.Param0WriteEnabled1=true
    Module.Param0WriteEnabled2=true
    Module.Param1WriteEnabled=true
    Module.Param1WriteEnabled1=true
    Module.Param1WriteEnabled2=true
    Module.Param2WriteEnabled=true
    Module.Param2WriteEnabled1=true
    Module.Param2WriteEnabled2=true
    Module.Param3WriteEnabled=true
    Module.Param3WriteEnabled1=true
    Module.Param3WriteEnabled2=true
    Write Parameter Index 1=true
    Write Parameter Index 2=true
   )
   /SetVariables(Particles.Ribbon No Sparks Params 0:Vector4f=&MakeVector4(
     X=&FloatFromCurve(
      FloatCurve=DataInterface() {propertyValues {Curve {keys [
          {arriveTangent=4.4937, interpMode=RCIM_Cubic, leaveTangent=4.4937, tangentMode=RCTM_User, time=0, value=0}
          {arriveTangent=1.4506, interpMode=RCIM_Cubic, leaveTangent=1.4506, tangentMode=RCTM_User, time=0.2, value=0.625}
          {arriveTangent=0.0378, interpMode=RCIM_Cubic, leaveTangent=0.0378, tangentMode=RCTM_User, time=1, value=1}
         ]}}}
      Scale Curve=2
     )
     Y=&FloatFromCurve(
      FloatCurve=DataInterface() {propertyValues {Curve {keys [
          {arriveTangent=-0.018, interpMode=RCIM_Cubic, leaveTangent=-0.018, tangentMode=RCTM_User, time=0, value=0}
          {arriveTangent=1.4018, interpMode=RCIM_Cubic, leaveTangent=1.4018, tangentMode=RCTM_User, time=1, value=1}
         ]}}}
      Scale Curve=6
     )
     Z=0
     W=0
    )) {isSetParameters=true, module=SetVariables_#8}
   /SetVariables(Particles.Ribbon No Sparks Params 1:Vector4f=&MakeVector4(
     X=1
     Y=&FloatFromCurve(
      FloatCurve=DataInterface() {propertyValues {Curve {keys [
          {arriveTangent=-6.499, interpMode=RCIM_Cubic, leaveTangent=-6.499, tangentMode=RCTM_User, time=0, value=3}
          {interpMode=RCIM_Cubic, tangentMode=RCTM_User, time=1, value=1}
         ]}}}
      Scale Curve=2
     )
     Z=0.5
     W=&FloatFromCurve(
      FloatCurve=DataInterface() {propertyValues {Curve {keys [
          {arriveTangent=29.4837, interpMode=RCIM_Cubic, leaveTangent=29.4837, tangentMode=RCTM_User, time=0, value=0}
          {interpMode=RCIM_Cubic, tangentMode=RCTM_User, time=0.1, value=1}
          {interpMode=RCIM_Cubic, tangentMode=RCTM_User, time=0.7, value=1}
          {interpMode=RCIM_Cubic, tangentMode=RCTM_User, time=1, value=0}
         ]}}}
      Scale Curve=10
     )
    )) {isSetParameters=true, module=SetVariables_#9}
   /SetVariables(Particles.Ribbon No Sparks Params 2:Vector4f=&MakeVector4(
     X=1
     Y=&FloatFromCurve(
      FloatCurve=DataInterface() {propertyValues {Curve {keys [
          {arriveTangent=0.8935, interpMode=RCIM_Cubic, leaveTangent=0.8935, tangentMode=RCTM_User, time=0.1, value=0}
          {arriveTangent=1.1556, interpMode=RCIM_Cubic, leaveTangent=1.1556, tangentMode=RCTM_User, time=1, value=1}
         ]}}}
      Scale Curve=5
     )
     Z=1
     W=0.1
    )) {isSetParameters=true, module=SetVariables_#10}
  ]
  rendererClasses [&NiagaraMeshRendererProperties]
  renderers [
   {rendererClass=&NiagaraMeshRendererProperties, rendererIndex=0}
   {rendererClass=&NiagaraMeshRendererProperties, rendererIndex=1}
  ]
 }
 Mesh_Ribbons_Optimized {
  emitterUpdateScript [
   &EmitterState() {module=EmitterState001}
   &SpawnPerFrame(Spawn:NiagaraBool=$System.DeathSpawn, Use Spawn Probability=false)
  ]
  particleSpawnScript [
   &InitializeParticle(
    Color=$System.SystemColor
    Lifetime=&RandomRangeFloat(Minimum=2.5, Maximum=3)
    Mesh Scale Mode=::NewEnumerator3
    Position=$System.DeathPosition
    Position Mode=::NewEnumerator1
    Sprite Size Mode=::NewEnumerator3
    Uniform Sprite Size=50
    UsePositionOffset=false
    Write Material Random=true
   )
   &InitialMeshOrientation(Use Rotation Vector=false)
   /SetVariables(Particles.Ribbon No Sparks Orientation:Quat4f=&RandomQuaternion()) {isSetParameters=true, module=SetVariables_#7}
  ]
  particleUpdateScript [
   &ParticleState()
   &ScaleMeshSize(Scale Factor=&VectorFromFloat(Value=$System.BurstMeshSize))
   &ScaleColor(
    Scale Alpha=&FloatFromCurve(FloatCurve=DataInterface() {propertyValues {Curve {keys [
         {arriveTangent=0.0543, interpMode=RCIM_Cubic, leaveTangent=0.0543, tangentMode=RCTM_User, time=0, value=0}
         {arriveTangent=0.0595, interpMode=RCIM_Cubic, leaveTangent=0.0595, tangentMode=RCTM_User, time=1, value=1}
        ]}}})
    Scale RGB=&VectorFromFloat(Value=&FloatFromCurve(FloatCurve=DataInterface() {propertyValues {Curve {keys [
          {arriveTangent=-33.3256, interpMode=RCIM_Cubic, leaveTangent=-33.3256, tangentMode=RCTM_User, time=0, value=5}
          {arriveTangent=-0.2459, interpMode=RCIM_Cubic, leaveTangent=-0.2459, tangentMode=RCTM_User, time=0.25, value=1}
         ]}}}))
    ScaleA=true
    ScaleRGB=true
   )
   &DynamicMaterialParameters(
    Index 0 Param 1=&FloatFromCurve(
     FloatCurve=DataInterface() {propertyValues {Curve {keys [
         {arriveTangent=4.4937, interpMode=RCIM_Cubic, leaveTangent=4.4937, tangentMode=RCTM_User, time=0, value=0}
         {arriveTangent=1.4506, interpMode=RCIM_Cubic, leaveTangent=1.4506, tangentMode=RCTM_User, time=0.2, value=0.625}
         {arriveTangent=0.0378, interpMode=RCIM_Cubic, leaveTangent=0.0378, tangentMode=RCTM_User, time=1, value=1}
        ]}}}
     Scale Curve=3
    )
    Index 0 Param 2=&FloatFromCurve(
     FloatCurve=DataInterface() {propertyValues {Curve {keys [
         {arriveTangent=-0.018, interpMode=RCIM_Cubic, leaveTangent=-0.018, tangentMode=RCTM_User, time=0, value=0}
         {arriveTangent=1.4018, interpMode=RCIM_Cubic, leaveTangent=1.4018, tangentMode=RCTM_User, time=1, value=1}
        ]}}}
     Scale Curve=2
    )
    Index 0 Param 3=-0.2
    Index 1 Param 1=1
    Index 1 Param 2=&FloatFromCurve(
     FloatCurve=DataInterface() {propertyValues {Curve {keys [
         {arriveTangent=-6.499, interpMode=RCIM_Cubic, leaveTangent=-6.499, tangentMode=RCTM_User, time=0, value=3}
         {interpMode=RCIM_Cubic, tangentMode=RCTM_User, time=1, value=1}
        ]}}}
     Scale Curve=3
    )
    Index 1 Param 3=0.5
    Index 1 Param 4=&FloatFromCurve(
     FloatCurve=DataInterface() {propertyValues {Curve {keys [
         {arriveTangent=0.0351, interpMode=RCIM_Cubic, leaveTangent=0.0351, tangentMode=RCTM_User, time=0, value=0}
         {interpMode=RCIM_Cubic, tangentMode=RCTM_User, time=0.7, value=1}
         {interpMode=RCIM_Cubic, tangentMode=RCTM_User, time=1, value=0}
        ]}}}
     Scale Curve=125
    )
    Index 2 Param 1=1
    Index 2 Param 2=&FloatFromCurve(
     FloatCurve=DataInterface() {propertyValues {Curve {keys [
         {arriveTangent=0.0495, interpMode=RCIM_Cubic, leaveTangent=0.0495, tangentMode=RCTM_User, time=0, value=0}
         {arriveTangent=1.1556, interpMode=RCIM_Cubic, leaveTangent=1.1556, tangentMode=RCTM_User, time=1, value=1}
        ]}}}
     Scale Curve=5
    )
    Index 2 Param 3=1
    Index 2 Param 4=0.35
    Module.Param0WriteEnabled=true
    Module.Param0WriteEnabled1=true
    Module.Param0WriteEnabled2=true
    Module.Param1WriteEnabled=true
    Module.Param1WriteEnabled1=true
    Module.Param1WriteEnabled2=true
    Module.Param2WriteEnabled=true
    Module.Param2WriteEnabled1=true
    Module.Param2WriteEnabled2=true
    Module.Param3WriteEnabled=true
    Module.Param3WriteEnabled1=true
    Module.Param3WriteEnabled2=true
    Write Parameter Index 1=true
    Write Parameter Index 2=true
   )
   /SetVariables(Particles.Ribbon No Sparks Params 0:Vector4f=&MakeVector4(
     X=&FloatFromCurve(
      FloatCurve=DataInterface() {propertyValues {Curve {keys [
          {arriveTangent=4.4937, interpMode=RCIM_Cubic, leaveTangent=4.4937, tangentMode=RCTM_User, time=0, value=0}
          {arriveTangent=1.4506, interpMode=RCIM_Cubic, leaveTangent=1.4506, tangentMode=RCTM_User, time=0.2, value=0.625}
          {arriveTangent=0.0378, interpMode=RCIM_Cubic, leaveTangent=0.0378, tangentMode=RCTM_User, time=1, value=1}
         ]}}}
      Scale Curve=2
     )
     Y=&FloatFromCurve(
      FloatCurve=DataInterface() {propertyValues {Curve {keys [
          {arriveTangent=-0.018, interpMode=RCIM_Cubic, leaveTangent=-0.018, tangentMode=RCTM_User, time=0, value=0}
          {arriveTangent=1.4018, interpMode=RCIM_Cubic, leaveTangent=1.4018, tangentMode=RCTM_User, time=1, value=1}
         ]}}}
      Scale Curve=6
     )
     Z=0
     W=0
    )) {isSetParameters=true, module=SetVariables_#8}
   /SetVariables(Particles.Ribbon No Sparks Params 1:Vector4f=&MakeVector4(
     X=1
     Y=&FloatFromCurve(
      FloatCurve=DataInterface() {propertyValues {Curve {keys [
          {arriveTangent=-6.499, interpMode=RCIM_Cubic, leaveTangent=-6.499, tangentMode=RCTM_User, time=0, value=3}
          {interpMode=RCIM_Cubic, tangentMode=RCTM_User, time=1, value=1}
         ]}}}
      Scale Curve=2
     )
     Z=0.5
     W=&FloatFromCurve(
      FloatCurve=DataInterface() {propertyValues {Curve {keys [
          {arriveTangent=29.4837, interpMode=RCIM_Cubic, leaveTangent=29.4837, tangentMode=RCTM_User, time=0, value=0}
          {interpMode=RCIM_Cubic, tangentMode=RCTM_User, time=0.1, value=1}
          {interpMode=RCIM_Cubic, tangentMode=RCTM_User, time=0.7, value=1}
          {interpMode=RCIM_Cubic, tangentMode=RCTM_User, time=1, value=0}
         ]}}}
      Scale Curve=10
     )
    )) {isSetParameters=true, module=SetVariables_#9}
   /SetVariables(Particles.Ribbon No Sparks Params 2:Vector4f=&MakeVector4(
     X=1
     Y=&FloatFromCurve(
      FloatCurve=DataInterface() {propertyValues {Curve {keys [
          {arriveTangent=0.8935, interpMode=RCIM_Cubic, leaveTangent=0.8935, tangentMode=RCTM_User, time=0.1, value=0}
          {arriveTangent=1.1556, interpMode=RCIM_Cubic, leaveTangent=1.1556, tangentMode=RCTM_User, time=1, value=1}
         ]}}}
      Scale Curve=5
     )
     Z=1
     W=0.1
    )) {isSetParameters=true, module=SetVariables_#10}
  ]
  rendererClasses [&NiagaraMeshRendererProperties]
  renderers [
   {rendererClass=&NiagaraMeshRendererProperties, rendererIndex=0}
   {rendererClass=&NiagaraMeshRendererProperties, rendererIndex=1}
  ]
 }
 Ribbon_Smoke {
  emitterUpdateScript [
   &EmitterState(
    Enable Distance Culling=true
    Life Cycle Mode=::NewEnumerator1
    Loop Behavior=::NewEnumerator1
    MaxDistance=2500000
    Module.bUseMaxDistance=true
    Module.bUseMinDistance=false
    Scalability Mode=::NewEnumerator1
    UseLoopDelay=false
   )
   &SpawnParticlesFromOtherEmitter(
    Attribute Reader=DataInterface() {propertyValues {EmitterBinding {bindingMode=Other, emitterName=MainSource}}}
    Spawn Rate=15
    Spawn Rate Per Particle Cap=200
    Use Maximum Age=false
    Use Minimum Age=false
   ) {module=SpawnParticlesFromOtherEmitter001}
  ]
  particleSpawnScript [
   &InitializeRibbon(
    Position Mode:ENiagara_PositionInitializationMode=::NewEnumerator1
    Ribbon Width=&FloatFromCurve(FloatCurve=DataInterface() {propertyValues {Curve {keys [<at 1 1 RCIM_Cubic>,<at 2 0>]}}}, CurveIndex=$Emitter.Age, Scale Curve=35)
    UsePositionOffset=false
    Write Material Random=true
   ) {module=InitializeRibbon001}
   &SampleParticlesFromOtherEmitter(
    Age Sampling:ENiagara_DirectReadSamplingMode=::NewEnumerator0
    Applied Velocity Scale=0
    Color Sampling=::NewEnumerator0
    Lifetime Application:ENiagara_LifetimeInheritanceOptions=::NewEnumerator1
    Lifetime Sampling:ENiagara_DirectReadSamplingMode=::NewEnumerator0
    Particle ID Sampling=::NewEnumerator1
    Reject Particle If Sampled Index Is Invalid=false
    Sprite Size Sampling:ENiagara_DirectReadSamplingMode=::NewEnumerator0
    Velocity Sampling=::NewEnumerator0
   )
   &AddVelocity(Velocity=&RandomRangeVector(Minimum=<0 -5 -5>, Maximum=<-1 5 5>), Velocity Speed Scale=17)
   &ApplyInitialForces()
  ]
  particleUpdateScript [
   &ParticleState()
   &DynamicMaterialParameters(
    Index 0 Param 1=-0.2
    Index 0 Param 2=1
    Index 0 Param 4=&FloatFromCurve(FloatCurve=DataInterface() {propertyValues {Curve {keys [
         <at 0 1 RCIM_Cubic>
         {arriveTangent=-0.3679, interpMode=RCIM_Cubic, leaveTangent=-0.3679, tangentMode=RCTM_User, time=1.6, value=0.7}
         {arriveTangent=-0.0204, tangentMode=RCTM_User, time=3, value=0}
        ]}}})
    Module.Param0WriteEnabled=true
    Module.Param1WriteEnabled=true
    Module.Param2WriteEnabled=true
    Module.Param3WriteEnabled=true
   )
   &ScaleRibbonWidth(Ribbon Width Scale=&FloatFromCurve(
     FloatCurve=DataInterface() {propertyValues {Curve {keys [{arriveTangent=-1.2246, interpMode=RCIM_Cubic, leaveTangent=-1.2246, tangentMode=RCTM_User, time=0.1, value=1}, <at 2 0>]}}}
     Scale Curve=7
    )) {enabled=false, module=ScaleRibbonWidth001}
   &ScaleRibbonWidth(Ribbon Width Scale=&FloatFromCurve(
     FloatCurve=DataInterface() {propertyValues {Curve {keys [
         {arriveTangent=0.0097, interpMode=RCIM_Cubic, leaveTangent=0.0097, tangentMode=RCTM_User, time=0.1, value=0.5}
         {arriveTangent=0.324, interpMode=RCIM_Cubic, leaveTangent=0.324, tangentMode=RCTM_User, time=0.9, value=0.875}
         <at 2 1>
        ]}}}
     Scale Curve=8
    ))
   /Niagara/Modules/Update/Forces/CurlNoiseForce(
    Noise Frequency=0.01
    Noise Strength=&Multiply_Float(
     A=10
     B=&FloatFromCurve(
      FloatCurve=DataInterface() {propertyValues {Curve {keys [
          {arriveTangent=1.9891, interpMode=RCIM_Cubic, leaveTangent=1.9891, tangentMode=RCTM_User, time=0, value=0}
          {interpMode=RCIM_Cubic, tangentMode=RCTM_User, time=0.1, value=1}
          {tangentMode=RCTM_User, time=1, value=0}
         ]}}}
      Scale Curve=5
     )
    )
    Pan Noise=true
   )
   &ScaleColor(
    Scale Alpha=&FloatFromCurve(FloatCurve=DataInterface() {propertyValues {Curve {keys [
         {arriveTangent=9.0725, interpMode=RCIM_Cubic, leaveTangent=9.0725, tangentMode=RCTM_User, time=0, value=0.01}
         {arriveTangent=-0.0314, interpMode=RCIM_Cubic, leaveTangent=-0.0314, tangentMode=RCTM_User, time=0.1, value=1}
         {arriveTangent=-0.0405, interpMode=RCIM_Cubic, leaveTangent=-0.0405, tangentMode=RCTM_User, time=1, value=0}
        ]}}})
    Scale RGB=&VectorFromFloat(Value=1)
    ScaleA=true
    ScaleRGB=true
   )
   &SolveForcesAndVelocity(Clamp Velocity=false, Limit Acceleration=false)
  ]
  rendererClasses [&NiagaraRibbonRendererProperties]
  renderers [{rendererClass=&NiagaraRibbonRendererProperties}]
 }
 Trail_Embers {
  emitterUpdateScript [
   &EmitterState(Life Cycle Mode=::NewEnumerator1, Loop Behavior=::NewEnumerator1, Loop Duration=2, UseLoopDelay=false)
   &SpawnParticlesFromOtherEmitter(
    Attribute Reader=DataInterface() {propertyValues {EmitterBinding {bindingMode=Other, emitterName=MainSource}}}
    Maximum Age Spawn Threshold=2
    Minimum Age Spawn Threshold=0
    Spawn Rate=50
    Spawn Rate Per Particle Cap=125
    Use Maximum Age=false
    Use Minimum Age=false
   )
  ]
  particleSpawnScript [
   &InitializeParticle(
    Color Mode=::NewEnumerator0
    Lifetime Max=0.45
    Lifetime Min=0.25
    Lifetime Mode=::NewEnumerator1
    Sprite Rotation Mode=::NewEnumerator2
    Sprite Size Mode=::NewEnumerator1
    Uniform Sprite Size Max=50
    Uniform Sprite Size Min=50
    UsePositionOffset=false
    Write Material Random=true
   )
   &SampleParticlesFromOtherEmitter(
    Applied Velocity Scale:NiagaraFloat=&FloatFromCurve(
     FloatCurve=DataInterface() {propertyValues {Curve {keys [{arriveTangent=-1.801, interpMode=RCIM_Cubic, leaveTangent=-1.801, tangentMode=RCTM_User, time=0, value=1}, <at 1 0>]}}}
     Scale Curve=0.5
    )
    Apply Sampled Alpha=false
    Color Sampling=::NewEnumerator0
    Maximum Sampled Particle Age=2
    Minimum Sampled Particle Age=0
    Module.MaxSampledAgeCutoff=true
    Module.MinSampledAgeCutoff=true
    Reject Particle By Sampled Age=true
    Velocity Sampling=::NewEnumerator0
   )
   &AddVelocity(
    Rotation Coordinate Space=::World
    Velocity=&RandomRangeVector(Minimum=<-500 -500 0>, Maximum=<500 500 0>)
   )
   &SphereLocation(
    MaskSpawnGroup=false
    Override Seed=false
    Sphere Radius=50
    Surface Only=false
    UseNonUniformScale=false
    UseOffset=false
    UseSphereOrientation=false
   )
  ]
  particleUpdateScript [
   &ParticleState()
   &Drag(Drag=&RandomRangeFloat(Minimum=2, Maximum=3), Use Linear Drag=true, Use Rotational Drag=false)
   &SolveForcesAndVelocity(Clamp Velocity=false, Limit Acceleration=false)
   &ScaleSpriteSize(
    Uniform Curve Scale=0.5
    Uniform Curve Sprite Scale=DataInterface() {propertyValues {Curve {keys [
        {arriveTangent=2.9868, interpMode=RCIM_Cubic, leaveTangent=2.9868, tangentMode=RCTM_User, time=0, value=1.125}
        {arriveTangent=0.275, interpMode=RCIM_Cubic, leaveTangent=0.275, time=0.4, value=2.3}
        {arriveTangent=-0.0405, interpMode=RCIM_Cubic, leaveTangent=-0.0405, tangentMode=RCTM_User, time=1, value=1.4}
       ]}}}
   )
   &ScaleColor(
    Scale Alpha=&FloatFromCurve(FloatCurve=DataInterface() {propertyValues {Curve {keys [{arriveTangent=-1.933, interpMode=RCIM_Cubic, leaveTangent=-1.933, tangentMode=RCTM_User, time=0, value=1}, <at 1 0>]}}})
    Scale RGB=&VectorFromFloat(Value=&FloatFromCurve(
      FloatCurve=DataInterface() {propertyValues {Curve {keys [{arriveTangent=-234.3046, interpMode=RCIM_Cubic, leaveTangent=-234.3046, tangentMode=RCTM_User, time=0, value=100}, <at 1 0>]}}}
      Scale Curve=10
     ))
    ScaleA=true
    ScaleRGB=true
   )
  ]
  rendererClasses [&NiagaraSpriteRendererProperties]
  renderers [{rendererClass=&NiagaraSpriteRendererProperties}]
 }
}
summary {
 systemName=NS_Sparks_Firework_Sphere_Double_Optimized_High
 emitters [
  {bEnabled=false, emitterName=Light_Glow, rendererClasses [&NiagaraSpriteRendererProperties]}
  {emitterName=MainSource, rendererClasses [&NiagaraSpriteRendererProperties]}
  {bEnabled=false, emitterName=FIrework_Blast, rendererClasses [&NiagaraMeshRendererProperties]}
  {bEnabled=false, emitterName=Mesh_Ribbons, rendererClasses [&NiagaraMeshRendererProperties]}
  {emitterName=Ribbon_Smoke, rendererClasses [&NiagaraRibbonRendererProperties]}
  {emitterName=Trail_Embers, rendererClasses [&NiagaraSpriteRendererProperties]}
  {emitterName=FIrework_Blast_Optimized, rendererClasses [&NiagaraMeshRendererProperties]}
  {emitterName=Mesh_Ribbons_Optimized, rendererClasses [&NiagaraMeshRendererProperties]}
  {emitterName=Light_Glow_Optimized, rendererClasses [&NiagaraSpriteRendererProperties]}
 ]
 userVariables [User.HeightMult {type=NiagaraFloat, defaultValue=1, description=""}]
}
